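# Material function for Hera cliff rocks, layering rock, optional sediment, a custom top layer and landscape matching.
# Base rock uses baked UVs or dithered world-space triplanar mapping; stacked switchable min/max masks drive layer blends.
# Curvature, dirt, biome color, RVT blending and waterline follow, then displacement, shared WPO effects and debug views.
# Quality, platform, shading-path and Nanite switches vary detail, including temporally dithered crack normals on high-end platforms.
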MaterialFunction /Hera/Terrain/Cliffs/Materials/MF_Hera_CliffRock
graph {
 properties {description=Checker}
 nodes [
  n1 = GetMaterialAttributes(None=NamedRerouteUsage()) {attributeGetTypes [#1]}
  n2 = VectorParameter() {group="2_Dirt/Curvature (L2_MSK as input)", parameterName=Curvature_Color, sortPriority=50, defaultValue=<rgba 0.4115 0.4115 0.4115 1>}
  n3 = Add(
   A=ScalarParameter() {group="2_Dirt/Curvature (L2_MSK as input)", parameterName=Curvature Bias, sliderMax=1, sortPriority=50}
   B=Multiply(
    A=NamedRerouteUsage()
    B=ScalarParameter() {defaultValue=1, group="2_Dirt/Curvature (L2_MSK as input)", parameterName=Curvature_Add Soft Curvature, sortPriority=50}
   )
  )
  n4 = NamedRerouteUsage()
  n5 = Constant()
  n6 = NamedRerouteUsage()
  n7 = NamedRerouteUsage()
  n8 = NamedRerouteUsage()
  n9 = Max(
   A=StaticSwitchParameter(True=StaticSwitch(True=@n4, False=@n5, Value=@n6), False=@n5) {defaultValue=true, group="2_Dirt/Curvature (L2_MSK as input)", parameterName=Curvature_L3 Mask, sortPriority=50}
   B=StaticSwitchParameter(True=StaticSwitch(True=@n7, False=@n5, Value=@n8), False=@n5) {defaultValue=true, group="2_Dirt/Curvature (L2_MSK as input)", parameterName=Curvature_L4 Mask, sortPriority=50}
  )
  n10 = Multiply(A=@n2.RGB, B=Multiply(A=Saturate(None=@n3), B=Subtract(B=@n9)))
  n11 = StaticSwitchParameter(
   True=SetMaterialAttributes!(MaterialAttributes=@n1.MaterialAttributes, BaseColor=Add(A=@n1.BaseColor, B=@n10)) {attributeSetTypes [#1]}
   False=@n1.MaterialAttributes
  ) {group="2_Dirt/Curvature (L2_MSK as input)", parameterName=Curvature, sortPriority=50}
  n12 = GetMaterialAttributes(None=@n11) {attributeGetTypes [#1]}
  n13 = GetMaterialAttributes(None=MaterialFunctionCall!(
    Material=SetMaterialAttributes(BaseColor=@n12.BaseColor) {attributeSetTypes [#1]}
    TerrainMasks_QuadA=NamedRerouteUsage()
    TerrainMasks_QuadB=NamedRerouteUsage()
    TerrainMasks_QuadC=NamedRerouteUsage()
    TerrainMasks_QuadD=NamedRerouteUsage()
    MossMask=Constant()
   ) {materialFunction=/Hera/Environment/Terrain/Landscape/Materials/Layers/BiomeColor/MF_Hera_Cliff_BiomeColor}.Result) {attributeGetTypes [#1]}
  n14 = Constant()
  n15 = Max(
   A=StaticSwitch(True=NamedRerouteUsage(), False=@n14, Value=NamedRerouteUsage())
   B=StaticSwitch(True=NamedRerouteUsage(), False=@n14, Value=NamedRerouteUsage())
  )
  n16 = Multiply(A=@n15, B=@n15)
  n17 = StaticSwitchParameter(
   True=SetMaterialAttributes(MaterialAttributes=@n12.MaterialAttributes, BaseColor=LinearInterpolate(A=@n13.BaseColor, B=@n12.BaseColor, Alpha=@n16)) {attributeSetTypes [#1]}
   False=@n11
  ) {
   desc=Set this to true only if biome colors are needed to match the cliffs
   group=BiomeColors
   parameterName=UseCliffBiomeColors
   sortPriority=-999
  }
  n18 = GetMaterialAttributes(None=@n17) {attributeGetTypes [#1]}
  n19 = VectorParameter() {
   group="2_Dirt/Curvature (L2_MSK as input)"
   parameterName=Dirt_AO Color
   sortPriority=101
   channelNames {a=Linear vs Mult, b="", g="", r=""}
   defaultValue=<rgba 0.2691 0.1783 0.1117 0.5>
  }
  n20 = QualitySwitch(Default=@n19, Epic=@n19)
  n21 = NamedRerouteUsage()
  n22 = ScalarParameter() {group="2_Dirt/Curvature (L2_MSK as input)", parameterName=Dirt_Bias, sortPriority=105}
  n23 = ScalarParameter() {group="2_Dirt/Curvature (L2_MSK as input)", parameterName="Use Non-Epic Colors? (Curvature/Dirt)", sliderMax=1, sortPriority=1}
  n24 = Saturate(None=@n23)
  n25 = LinearInterpolate(A=@n22, B=ScalarParameter() {group="2_Dirt/Curvature (L2_MSK as input)", parameterName=Dirt_Bias Non-Epic, sortPriority=105}, Alpha=@n24)
  n26 = ScalarParameter() {defaultValue=1, group="2_Dirt/Curvature (L2_MSK as input)", parameterName=Dirt_Contrast, sortPriority=105}
  n27 = ScalarParameter() {defaultValue=1, group="2_Dirt/Curvature (L2_MSK as input)", parameterName=Dirt_Contrast Non-Epic, sortPriority=105}
  n28 = MaterialFunctionCall(
   "Xb1,PS4"=@n27
   "PC-High&&Epic,XSX,PS5"=@n27
   "PC-Perf,Android"=ScalarParameter() {defaultValue=1, group="2_Dirt/Curvature (L2_MSK as input)", parameterName=Dirt_Contrast Mobile, sortPriority=105}
   Switch=@n27
   "PC-Low(NotPerf)&&Medium"=@n27
  ) {materialFunction=&PlatformSwitch_1D}
  n29 = Multiply(
   A=Add(A=OneMinus(None=@n21), B=QualitySwitch(Default=@n25, Epic=@n22))
   B=QualitySwitch(Default=LinearInterpolate(A=@n26, B=@n28.Result, Alpha=@n24), Epic=@n26)
  )
  n30 = NamedRerouteUsage()
  n31 = Constant()
  n32 = NamedRerouteUsage()
  n33 = NamedRerouteUsage()
  n34 = NamedRerouteUsage()
  n35 = Max(
   A=StaticSwitchParameter(True=StaticSwitch(True=@n30, False=@n31, Value=@n32), False=@n31) {defaultValue=true, group="2_Dirt/Curvature (L2_MSK as input)", parameterName=Dirt_L3 Mask, sortPriority=55}
   B=StaticSwitchParameter(True=StaticSwitch(True=@n33, False=@n31, Value=@n34), False=@n31) {defaultValue=true, group="2_Dirt/Curvature (L2_MSK as input)", parameterName=Dirt_L4 Mask, sortPriority=55}
  )
  n36 = LinearInterpolate(
   A=@n18.BaseColor
   B=LinearInterpolate(A=Multiply(A=@n20, B=@n18.BaseColor), B=@n19, Alpha=@n19)
   Alpha=Multiply(A=Saturate(None=@n29), B=Subtract(B=@n35))
  )
  n37 = NamedRerouteUsage()
  n38 = NamedRerouteUsage()
  n39 = MaterialFunctionCall!(
   In=StaticSwitchParameter(
    True=SetMaterialAttributes(MaterialAttributes=@n18.MaterialAttributes, BaseColor=@n36) {attributeSetTypes [#1]}
    False=@n18.MaterialAttributes
   ) {group="2_Dirt/Curvature (L2_MSK as input)", parameterName=Dirt, sortPriority=100}
   "RVT Height Map (Roughness)"=StaticSwitch(True=Multiply(A=@n37, B=@n38), False=@n38, Value=NamedRerouteUsage())
  ) {desc="RVT ", materialFunction=&/MF_Hera_RVTBlendingCliffRock}
  n40 = MaterialFunctionCall!(RVT=NamedRerouteUsage(), WaterInput=@n39.Result) {materialFunction=/Game/Global/Base4Layers_2022/MaterialFunctions/MF_Rocks_2023_Waterline}
  n41 = MaterialFunctionCall!(Material=@n39.Result) {materialFunction=/Game/Athena/Helios/Environments/Landscape/Materials/MaterialFunctions_Utilities/MF_WetShore_Mobile}
  n42 = NamedRerouteUsage()
  n43 = Constant()
  n44 = NamedRerouteUsage()
  n45 = BlendMaterialAttributes(
   A=ShadingPathSwitch(Default=@n40.BlendMaterialAttributes, Mobile=@n41.Result)
   B=@n39.Result
   Alpha=StaticSwitch(True=NamedRerouteUsage(), False=StaticSwitch(True=@n42, False=@n43, Value=@n44), Value=NamedRerouteUsage())
  ) {vertexAttributeBlendType=UseB}
  n46 = GetMaterialAttributes(None=MaterialProxyReplace(
    Realtime=StaticSwitchParameter(True=@n45, False=@n39.Result) {group=1_Rock/Cliff Addons, parameterName=Use Waterline}
    MaterialProxy=@n39.Result
   )) {attributeGetTypes [#2]}
  n47 = VertexColor()
  n48 = SetMaterialAttributes!(
   MaterialAttributes=@n46.MaterialAttributes
   Displacement=LinearInterpolate(A=Constant() {r=0.5}, B=@n46.Opacity, Alpha=Power(Base=@n47.R) {constExponent=8})
  ) {attributeSetTypes [#3]}
  Composite!(MaterialAttributes=@n48) {subgraphName=Documentation Link}
  MaterialExpressionReroute_70 = Reroute(None=@n48)
  n49 = ScalarParameter() {
   desc=Documentation link
   group="0 (Expand)  Documentation (RMB - Copy display name) and Debug view"
   parameterName=https://confluence.it.epicgames.com/display/FORT/Fortnite+Base+Material+2022
  }
  n50 = SetMaterialAttributes!(MaterialAttributes=@MaterialExpressionReroute_70.MaterialAttributes, WorldPositionOffset=Multiply(A=@n49)) {attributeSetTypes [#4]}
  MaterialExpressionReroute_98 = Reroute(None=@n50)
  Composite(MaterialAttributes=@MaterialExpressionReroute_98.Output) {desc=Game Shared WPO and Fire Effect, subgraphName=WPO And Shared Functions}
  n51 = NamedRerouteUsage()
  n52 = NamedRerouteUsage()
  n53 = SetMaterialAttributes!(MaterialAttributes=NamedRerouteUsage(), EmissiveColor=VertexColor()) {attributeSetTypes [#5]}
  MaterialExpressionReroute_99 = Reroute(None=@MaterialExpressionReroute_98.Output)
  n54 = MaterialFunctionCall!(In=MaterialFunctionCall(Material_In=@MaterialExpressionReroute_99.MaterialAttributes) {materialFunction=&MF_FN_DestructibleDamage}.Material_Out) {materialFunction=/Game/Global/MaterialFunctions/MF_WPOFloatBobAndRotate}
  MaterialExpressionReroute_100 = Reroute(None=@n54.Result)
  n55 = StaticSwitchParameter(
   True=SetMaterialAttributes(MaterialAttributes=NamedRerouteUsage(), EmissiveColor=NamedRerouteUsage()) {attributeSetTypes [#5]}
   False=StaticSwitchParameter(
    True=SetMaterialAttributes(MaterialAttributes=@n51, EmissiveColor=@n52) {attributeSetTypes [#5]}
    False=StaticSwitchParameter(True=@n53, False=@MaterialExpressionReroute_100.Result) {group="A_L2 (TD slot 2)", parameterName="[DEBUG] Vertex Colors", sortPriority=-10}
   ) {group=A_L4 Landscape Mask, parameterName="[DEBUG] L4 Masks", sortPriority=-1}
  ) {group=A_L3 Custom Mask, parameterName="[DEBUG] L3 Masks", sortPriority=-1}
  FunctionOutput(None=@n55)
  n56 = ScalarParameter() {group=Slurp, parameterName=Dirt Disable, sliderMax=1, sortPriority=2}
  n57 = ScalarParameter() {group=Slurp, parameterName=Dirt Disable Mobile, sliderMax=1, sortPriority=2}
  LinearInterpolate(Alpha=Saturate(None=MaterialFunctionCall("Xb1,PS4"=@n56, "PC-High&&Epic,XSX,PS5"=@n56, "PC-Perf,Android"=@n57, Switch=@n57, "PC-Low(NotPerf)&&Medium"=@n57) {materialFunction=&PlatformSwitch_1D}.Result))
  n58 = VectorParameter() {
   group="2_Dirt/Curvature (L2_MSK as input)"
   parameterName=Dirt_AO Color Non-Epic
   sortPriority=101
   channelNames {a=Linear vs Mult, b="", g="", r=""}
   defaultValue=<rgba 0.2691 0.1783 0.1117 0.5>
  }
  n59 = ScalarParameter() {group="2_Dirt/Curvature (L2_MSK as input)", parameterName="Use Non-Epic Colors? (Curvature/Dirt)", sliderMax=1, sortPriority=1}
  LinearInterpolate(A=@n19, B=@n58, Alpha=Saturate(None=@n59))
  n60 = StaticBoolParameter() {group=0_Rock/Cliff Main Settings, parameterName=L3_Use Custom Layer, sortPriority=5}
  n61 = NamedRerouteUsage()
  n62 = NamedRerouteUsage()
  n63 = NamedRerouteUsage()
  n64 = GetMaterialAttributes(None=StaticSwitch(
    True=BlendMaterialAttributes(A=@n61, B=@n62, Alpha=@n63) {vertexAttributeBlendType=UseB}
    False=@n61
    Value=NamedRerouteDeclaration(None=@n60) {name=L3_Use CustomTopLayer, nodeColor=<rgba 0.7109 1 0 1>}
   )) {attributeGetTypes [#6]}
  n65 = SetMaterialAttributes!(MaterialAttributes=@n64.MaterialAttributes, Normal=Transform(None=@n64.Normal)) {attributeSetTypes [#6]}
  MaterialFunctionCall!(In=@n65) {materialFunction=/Game/Environments/Helios/Foliage/Trees/MasterMaterials/MaterialFunctions/MF_WetlandsMoss_World}
  MaterialFunctionCall!("Xb1,PS4", "PC-High&&Epic,XSX,PS5", "PC-Perf,Android", Switch, "PC-Low(NotPerf)&&Medium") {materialFunction=/Game/Effects/Niagara/MasterMaterials/Functions/Common/PlatformSwitch_MA}
  MaterialFunctionCall!(In=@n65) {materialFunction=/Game/Environments/Asteria/Foliage/Trees/MasterMaterials/MaterialFunctions/MF_JungleMoss_World}
  Multiply(
   A=NamedRerouteUsage()
   B=LinearInterpolate(
    A=ScalarParameter() {defaultValue=0.75, parameterName=Dirt_Roughness Curvature Min}
    B=ScalarParameter() {defaultValue=2, parameterName=Dirt_Roughness Curvature Max}
    Alpha=NamedRerouteUsage()
   )
  )
  n66 = NamedRerouteUsage()
  n67 = StaticSwitchParameter(True=@n66, False=OneMinus(None=@n66)) {group=A_L3 Custom Mask, parameterName=L3_AO_Invert, sortPriority=31}
  n68 = NamedRerouteUsage()
  n69 = StaticSwitchParameter(
   True=StaticSwitchParameter(True=OneMinus(None=@n68), False=@n68) {desc=L3 Mask_World Z Normal, group=A_L3 Custom Mask, parameterName=L3_World Z Normal_Invert, sortPriority=11}
   False=Constant()
  ) {desc=L3 Mask_World Z Normal, group=A_L3 Custom Mask, parameterName="== L3_WORLD Z NORMAL MASK ==", sortPriority=10}
  n70 = NamedRerouteUsage()
  n71 = StaticSwitchParameter(True=@n70, False=OneMinus(None=@n70)) {group=A_L3 Custom Mask, parameterName=L3_VTX Color_Invert, sortPriority=21}
  n72 = Max(A=@n69, B=@n71)
  n73 = Min(A=@n69, B=@n71)
  n74 = StaticSwitchParameter(
   True=StaticSwitchParameter(
    True=@n71
    False=StaticSwitchParameter(True=@n72, False=@n73) {defaultValue=true, group=A_L3 Custom Mask, parameterName="L3_VTX Color_Blend (T=Max F=Min)", sortPriority=22}
   ) {group=A_L3 Custom Mask, parameterName=L3_VTX Color_Only, sortPriority=20}
   False=@n69
  ) {group=A_L3 Custom Mask, parameterName="== L3_VTX COLOR MASK ==", sortPriority=20}
  n75 = StaticSwitchParameter(True=Max(A=@n74, B=@n67), False=Min(A=@n74, B=@n67)) {defaultValue=true, desc=Additive/Subtractive Blending Mode switch, group=A_L3 Custom Mask, parameterName="L3_AO_Blend (T=Max F=Min)"}
  n76 = StaticSwitchParameter(True=StaticSwitchParameter(True=@n67, False=@n75) {group=A_L3 Custom Mask, parameterName=L3_AO_Only, sortPriority=30}, False=@n74) {group=A_L3 Custom Mask, parameterName="== L3_AO MASK ==", sortPriority=30}
  n77 = NamedRerouteUsage()
  n78 = StaticSwitchParameter(True=OneMinus(None=@n77), False=@n77) {group=A_L3 Custom Mask, parameterName=L3_Curvature_Invert, sortPriority=41}
  n79 = StaticSwitchParameter(True=Max(A=@n76, B=@n78), False=Min(A=@n76, B=@n78)) {
   defaultValue=true
   desc=Additive/Subtractive Blending Mode switch
   group=A_L3 Custom Mask
   parameterName="L3_Curvature_Blend (T=Max F=Min)"
   sortPriority=42
  }
  n80 = StaticSwitchParameter(
   True=StaticSwitchParameter(True=@n78, False=@n79) {desc=Additive/Subtractive Blending Mode switch, group=A_L3 Custom Mask, parameterName=L3_Curvature_Only, sortPriority=40}
   False=@n76
  ) {group=A_L3 Custom Mask, parameterName="== L3_CURVATURE MASK ==", sortPriority=40}
  n81 = NamedRerouteUsage()
  n82 = StaticSwitchParameter(True=OneMinus(None=@n81), False=@n81) {group=A_L3 Custom Mask, parameterName=L3_Painted_Invert, sortPriority=51}
  n83 = StaticSwitchParameter(True=Max(A=@n80, B=@n82), False=Min(A=@n80, B=@n82)) {
   defaultValue=true
   desc=Additive/Subtractive Blending Mode switch
   group=A_L3 Custom Mask
   parameterName="L3_Painted_Blend (T=Max F=Min)"
   sortPriority=52
  }
  n84 = StaticSwitchParameter(
   True=StaticSwitchParameter(True=@n82, False=@n83) {desc=Additive/Subtractive Blending Mode switch, group=A_L3 Custom Mask, parameterName=L3_Painted_Only, sortPriority=50}
   False=@n80
  ) {group=A_L3 Custom Mask, parameterName="== L3_PAINTED MASK ==", sortPriority=50}
  n85 = NamedRerouteUsage()
  n86 = StaticSwitchParameter(True=@n85, False=OneMinus(None=@n85)) {group=A_L3 Custom Mask, parameterName=L3_Z Gradient_Invert, sortPriority=61}
  n87 = StaticSwitchParameter(True=Max(A=@n84, B=@n86), False=Min(A=@n84, B=@n86)) {
   defaultValue=true
   desc=Additive/Subtractive Blending Mode switch
   group=A_L3 Custom Mask
   parameterName="L3_Z Gradient_Blend (T=Max F=Min)"
   sortPriority=62
  }
  n88 = StaticSwitchParameter(
   True=StaticSwitchParameter(True=@n86, False=@n87) {desc=Additive/Subtractive Blending Mode switch, group=A_L3 Custom Mask, parameterName=L3_Z Gradient_Only, sortPriority=60}
   False=@n84
  ) {group=A_L3 Custom Mask, parameterName="== L3_Z Gradient Mask ==", sortPriority=60}
  n89 = NamedRerouteUsage()
  n90 = StaticSwitchParameter(True=OneMinus(None=@n89), False=@n89) {group=A_L3 Custom Mask, parameterName=L3_Waterline_Invert, sortPriority=71}
  n91 = StaticSwitchParameter(True=Max(A=@n88, B=@n90), False=Min(A=@n88, B=@n90)) {
   defaultValue=true
   desc=Additive/Subtractive Blending Mode switch
   group=A_L3 Custom Mask
   parameterName="L3_Waterline_Blend (T=Max F=Min)"
   sortPriority=72
  }
  n92 = StaticSwitchParameter(
   True=StaticSwitchParameter(True=@n90, False=@n91) {desc=Additive/Subtractive Blending Mode switch, group=A_L3 Custom Mask, parameterName=L3_Waterline_Only, sortPriority=72}
   False=@n88
  ) {group=A_L3 Custom Mask, parameterName="== L3_WATERLINE MASK ==", sortPriority=70}
  n93 = NamedRerouteUsage()
  n94 = StaticSwitchParameter(True=OneMinus(None=@n93), False=@n93) {group=A_L3 Custom Mask, parameterName=L3_SedimentBands_Invert, sortPriority=81}
  n95 = StaticSwitchParameter(True=Max(A=@n92, B=@n94), False=Min(A=@n92, B=@n94)) {
   defaultValue=true
   desc=Additive/Subtractive Blending Mode switch
   group=A_L3 Custom Mask
   parameterName="L3_SedimentBands_Blend (T=Max F=Min)"
   sortPriority=82
  }
  n96 = StaticSwitchParameter(True=@n94, False=@n95) {desc=Additive/Subtractive Blending Mode switch, group=A_L3 Custom Mask, parameterName=L3_SedimentBands_Only, sortPriority=80}
  n97 = StaticSwitchParameter(True=@n96, False=@n92) {group=A_L3 Custom Mask, parameterName="== L3_SEDIMENT BANDS MASK ==", sortPriority=80}
  n98 = NamedRerouteUsage()
  n99 = StaticSwitchParameter(True=OneMinus(None=@n98), False=@n98) {group=A_L3 Custom Mask, parameterName=L3_WS Noise_Invert, sortPriority=101}
  n100 = StaticSwitchParameter(True=Max(A=@n97, B=@n99), False=Min(A=@n97, B=@n99)) {
   defaultValue=true
   desc=Additive/Subtractive Blending Mode switch
   group=A_L3 Custom Mask
   parameterName="L3_WS Noise_Blend (T=Max F=Min)"
   sortPriority=102
  }
  n101 = StaticSwitchParameter(
   True=StaticSwitchParameter(True=@n99, False=@n100) {desc=Additive/Subtractive Blending Mode switch, group=A_L3 Custom Mask, parameterName=L3_WS Noise_Only, sortPriority=100}
   False=@n97
  ) {group=A_L3 Custom Mask, parameterName="== L3_WS NOISE MASK ==", sortPriority=100}
  n102 = NamedRerouteUsage()
  n103 = StaticSwitchParameter(True=OneMinus(None=@n102), False=@n102) {group=A_L3 Custom Mask, parameterName=L3_L1 Height_Invert, sortPriority=111}
  n104 = StaticSwitchParameter(True=Max(A=@n101, B=@n103), False=Min(A=@n101, B=@n103)) {
   defaultValue=true
   desc=Additive/Subtractive Blending Mode switch
   group=A_L3 Custom Mask
   parameterName="L3_L1 Height_Blend (T=Max F=Min)"
   sortPriority=112
  }
  n105 = StaticSwitchParameter(True=StaticSwitchParameter(True=@n103, False=@n104) {group=A_L3 Custom Mask, parameterName=L3_L1 Height_Only, sortPriority=110}, False=@n101) {group=A_L3 Custom Mask, parameterName="== L3_L1 Height MASK ==", sortPriority=110}
  n106 = NamedRerouteUsage()
  n107 = StaticSwitchParameter(True=OneMinus(None=@n106), False=@n106) {group=A_L3 Custom Mask, parameterName=L3_L1 SMR_Invert, sortPriority=121}
  n108 = StaticSwitchParameter(True=Max(A=@n105, B=@n107), False=Min(A=@n105, B=@n107)) {
   defaultValue=true
   desc=Additive/Subtractive Blending Mode switch
   group=A_L3 Custom Mask
   parameterName="L3_L1 SMR_Blend (T=Max F=Min)"
   sortPriority=122
  }
  n109 = StaticSwitchParameter(True=StaticSwitchParameter(True=@n107, False=@n108) {group=A_L3 Custom Mask, parameterName=L3_L1 SMR_Only, sortPriority=120}, False=@n105) {group=A_L3 Custom Mask, parameterName="== L3_L1 SMR MASK ==", sortPriority=120}
  n110 = NamedRerouteUsage()
  n111 = StaticSwitchParameter(True=OneMinus(None=@n110), False=@n110) {group=A_L3 Custom Mask, parameterName=L3_Height_Invert, sortPriority=591}
  n112 = StaticSwitchParameter(True=Max(A=@n109, B=@n111), False=Min(A=@n109, B=@n111)) {
   defaultValue=true
   desc=Additive/Subtractive Blending Mode switch
   group=A_L3 Custom Mask
   parameterName="L3_Height_Blend (T=Max F=Min)"
   sortPriority=591
  }
  n113 = StaticSwitchParameter(True=StaticSwitchParameter(True=@n111, False=@n112) {group=A_L3 Custom Mask, parameterName=L3_Height_Only, sortPriority=590}, False=@n109) {group=A_L3 Custom Mask, parameterName="== L3_HEIGHT MASK ==", sortPriority=590}
  n114 = Multiply(A=StaticSwitchParameter(True=OneMinus(None=@n113), False=@n113) {group=A_L3 Custom Mask, parameterName=L3 Mask_Invert, sortPriority=0}) {constB=2, desc=2}
  n115 = Subtract(A=LinearInterpolate(
    A=ScalarParameter() {group="A_L3_Custom Layer (TD slot 3)", parameterName=L3_Custom H Min, sortPriority=580}
    B=ScalarParameter() {defaultValue=1, group="A_L3_Custom Layer (TD slot 3)", parameterName=L3Custom H Max, sortPriority=580}
    Alpha=NamedRerouteUsage()
   )) {desc=1}
  NamedRerouteDeclaration(None=Saturate(None=Add(A=@n114, B=@n115))) {name=L3 Mask Blended, nodeColor=<rgba 0 0.5898 1 1>}
  n116 = MaterialFunctionCall!(Grass_UseCustomSlopeMask, Grass_CustomSlopeMask) {materialFunction=/Hera/Environment/Terrain/Landscape/Materials/Layers/MF_Hera_StaticMeshLandscapeLayerMatch}
  NamedRerouteDeclaration(None=@n116.ChaparralGrassLayer) {name=Landscape_ChaparralGrass, nodeColor=<rgba 0.0312 0 1 1>}
  n117 = TextureCoordinate()
  n118 = TextureSampleParameter2D(UVs=@n117) {
   group="A_L2 (TD slot 2)"
   parameterName=Normals_Texture_2
   samplerSource=SSM_Wrap_WorldGroupSettings
   samplerType=SAMPLERTYPE_Normal
   texture=/Engine/EngineMaterials/FlatNormal
  }
  n119 = ViewProperty() {property=MEVP_TemporalSampleIndex}
  MaterialExpressionAdd_37 = Add(A=ScreenPosition().PixelPosition, B=AppendVector(A=@n119.Property, B=@n119.Property))
  n120 = Custom!(p=@MaterialExpressionAdd_37) {
   code="Mod( ((uint)(p.x) + 2 * (uint)(p.y)) , 5 )"
   outputType=CMOT_Float1
   inputs [{inputName=p, input {expression=&self:MaterialExpressionAdd_37}}]
  }
  n121 = ConstantBiasScale(None=@n120) {bias=-2, scale=0.1}
  n122 = ScalarParameter() {defaultValue=0.05, group=CustomCrack, parameterName=CustomCrack_Dither}
  n123 = Add(A=@n117, B=Multiply(A=@n121, B=@n122))
  n124 = TextureSampleParameter2D(UVs=@n123) {
   group="A_L2 (TD slot 2)"
   parameterName=Normals_Texture_2
   samplerSource=SSM_Wrap_WorldGroupSettings
   samplerType=SAMPLERTYPE_Normal
   texture=&FlatNormal
  }
  n125 = Multiply(A=@n122) {constB=-1}
  n126 = Add(A=@n117, B=Multiply(A=@n121, B=AppendVector(A=@n125, B=@n122)))
  n127 = TextureSampleParameter2D(UVs=@n126) {
   group="A_L2 (TD slot 2)"
   parameterName=Normals_Texture_2
   samplerSource=SSM_Wrap_WorldGroupSettings
   samplerType=SAMPLERTYPE_Normal
   texture=&FlatNormal
  }
  n128 = Add(A=@n124.RGB, B=@n127.RGB)
  NamedRerouteDeclaration(None=MaterialFunctionCall!(
    "Xb1,PS4"=@n118.RGB
    "PC-High&&Epic,XSX,PS5"=Normalize(VectorInput=@n128)
    "PC-Perf,Android"=@n118.RGB
    Switch=@n118.RGB
    "PC-Low(NotPerf)&&Medium"=@n118.RGB
   ) {materialFunction=/Game/Effects/Niagara/MasterMaterials/Functions/Common/PlatformSwitch_3D}.Result) {name=Crack_TextureNormals, nodeColor=<rgba 1 0.7031 0 1>}
  NamedRerouteDeclaration(None=@n116.TerrainMasks_QuadB) {name=TerrainMasks_QuadB, nodeColor=<rgba 0.5 0 1 1>}
  NamedRerouteDeclaration(None=@n116.TerrainMasks_QuadA) {name=TerrainMasks_QuadA, nodeColor=<rgba 0.5 0 1 1>}
  NamedRerouteDeclaration(None=@n116.TerrainMasks_QuadC) {name=TerrainMasks_QuadC, nodeColor=<rgba 0.5 0 1 1>}
  NamedRerouteDeclaration(None=@n116.TerrainMasks_QuadD) {name=TerrainMasks_QuadD, nodeColor=<rgba 0.5 0 1 1>}
  NamedRerouteDeclaration(None=@n39.RVT H) {name=RVT H, nodeColor=<rgba 1 0.5156 0 1>}
  n129 = MaterialFunctionCall!(
   Sediment MA=NamedRerouteUsage()
   IsLandscape=StaticBool()
   Sediment Normal Intensity=ScalarParameter() {defaultValue=1, group=Cliff, parameterName=Sediment_b | L2 Cliff Normal Intensity, sortPriority=307}
   Cliffs MA=NamedRerouteUsage()
   Sediment Wave Distiortion=NamedRerouteUsage()
   Sediment Normal B Mask=NamedRerouteUsage()
   Biom Mask=NamedRerouteUsage()
  ) {materialFunction=/Game/Athena/Helios/Environments/Landscape/Materials/MaterialFunctions_Utilities/MF_CliffSediment_Blend}
  n130 = NamedRerouteUsage()
  n131 = NamedRerouteUsage()
  n132 = ScalarParameter() {defaultValue=75, group="A_L1 (TD slot 1)", parameterName=L1 WS_Triplanar Fade_Low/Mobile, sortPriority=0}
  n133 = ScalarParameter() {defaultValue=10, group="A_L1 (TD slot 1)", parameterName=L1 WS_Triplanar Fade, sortPriority=0}
  n134 = MaterialFunctionCall!(
   "Normal (WS)"
   Texture Scale=NamedRerouteDeclaration(None=ScalarParameter() {defaultValue=2000, group="A_L1 (TD slot 1)", parameterName=L1 WS_Tile, sortPriority=0}) {name=L1 WS_Tile, nodeColor=<rgba 1 0 0.75 1>}
   Triplanar Fade=MaterialFunctionCall!(Low=@n132, High=@n133, Epic=@n133, --------, Mobile=@n132, Medium=@n133) {materialFunction=/FortniteBaseMaterials/Functions/MF_QualitySwitch_Scalar}.Result
   Derivatives Scale
  ) {materialFunction=/Game/Athena/Environments/MaterialFunctions/MF_TriplanarDithered}
  n135 = TextureSampleParameter2D!(UVs=@n134.UVs, "DDX(UVs)"=@n134.DDX, "DDY(UVs)"=@n134.DDY, Apply View MipBias) {
   group="A_L1 (TD slot 1)"
   mipValueMode=TMVM_Derivative
   parameterName=Diffuse
   samplerSource=SSM_Wrap_WorldGroupSettings
   sortPriority=10
   texture=/Game/Environments/Asteria/Terrain/Textures/T_Asteria_RockCliff_04_D
  }
  n136 = TextureSampleParameter2D(UVs=@n134.UVs, "DDX(UVs)"=@n134.DDX, "DDY(UVs)"=@n134.DDY) {
   group="A_L1 (TD slot 1)"
   mipValueMode=TMVM_Derivative
   parameterName=SpecularMasks
   samplerSource=SSM_Wrap_WorldGroupSettings
   samplerType=SAMPLERTYPE_Masks
   sortPriority=20
   texture=/Game/Environments/Asteria/Terrain/Textures/T_Asteria_RockCliff_04_S
  }
  n137 = TextureSampleParameter2D(UVs=@n134.UVs, "DDX(UVs)"=@n134.DDX, "DDY(UVs)"=@n134.DDY) {
   group="A_L1 (TD slot 1)"
   mipValueMode=TMVM_Derivative
   parameterName=Normals
   samplerSource=SSM_Wrap_WorldGroupSettings
   samplerType=SAMPLERTYPE_Normal
   sortPriority=30
   texture=/Game/Environments/Asteria/Terrain/Textures/T_Asteria_RockCliff_04_N
  }
  n138 = SetMaterialAttributes!(
   MaterialAttributes
   BaseColor=@n135.RGB
   Opacity=@n135.RGB
   Roughness=@n136.B
   Specular=@n136.B
   Metallic=@n136.B
   Normal=Transform(None=MaterialFunctionCall!(Tangent Normals=@n137.RGB, Vertex Normals, Triplanar Mask=@n134.Mask) {materialFunction=/Game/Athena/Environments/MaterialFunctions/MF_TriplanarNormalTangentToWorld}.World Normals) {transformSourceType=TRANSFORMSOURCE_World, transformType=TRANSFORM_Tangent}
  ) {attributeSetTypes [
    #1
    #2
    #7
    #8
    #9
    #6
   ]}
  n139 = Divide(A=WorldPosition() {worldPositionShaderOffset=WPT_ExcludeAllShaderOffsets}.XY, B=NamedRerouteUsage())
  n140 = TextureSampleParameter2D!(UVs=@n139, Apply View MipBias) {
   group="A_L1 (TD slot 1)"
   parameterName=Diffuse
   samplerSource=SSM_Wrap_WorldGroupSettings
   sortPriority=10
   texture=&T_Asteria_RockCliff_04_D
  }
  n141 = TextureSampleParameter2D(UVs=@n139) {
   group="A_L1 (TD slot 1)"
   parameterName=SpecularMasks
   samplerSource=SSM_Wrap_WorldGroupSettings
   samplerType=SAMPLERTYPE_Masks
   sortPriority=20
   texture=&T_Asteria_RockCliff_04_S
  }
  n142 = TextureSampleParameter2D(UVs=@n139) {
   desc=These normals aren't correct
   group="A_L1 (TD slot 1)"
   parameterName=Normals
   samplerSource=SSM_Wrap_WorldGroupSettings
   samplerType=SAMPLERTYPE_Normal
   sortPriority=30
   texture=&T_Asteria_RockCliff_04_N
  }
  n143 = SetMaterialAttributes(BaseColor=@n140.RGB, Opacity=@n140.RGB, Roughness=@n141.B, Specular=@n141.B, Metallic=@n141.B, Normal=@n142.RGB) {attributeSetTypes [
    #1
    #2
    #7
    #8
    #9
    #6
   ]}
  n144 = StaticBoolParameter() {
   defaultValue=true
   desc="On: L1 Use UV0\r\nOff: L1 Use UV1"
   group=0_Rock/Cliff Main Settings
   parameterName=L1 BakedProp_Use Displacement
   sortPriority=0
  }
  n145 = StaticSwitch(
   True=TextureCoordinate()
   False=TextureCoordinate() {coordinateIndex=1}
   Value=NamedRerouteDeclaration(None=@n144) {name=L1 Displacement, nodeColor=<rgba 0 1 0.7188 1>}
  )
  n146 = StaticSwitch(
   True=@n145
   False=Multiply(A=@n145, B=ScalarParameter() {defaultValue=1, group="A_L1 (TD slot 1)", parameterName=L1 BakedProp_Tile, sortPriority=-1})
   Value=NamedRerouteDeclaration(None=StaticBoolParameter() {desc="true = baked UV\r\nfalse= WS Cliffs", group=0_Rock/Cliff Main Settings, parameterName=IS Baked Prop, sortPriority=0}) {name=IsBakedProp, nodeColor=<rgba 1 0.75 0 1>}
  )
  n147 = TextureSampleParameter2D(UVs=NamedRerouteDeclaration(None=@n146) {name=L1 UV Baked, nodeColor=<rgba 0 1 0.8594 1>}) {
   group="A_L1 (TD slot 1)"
   parameterName=Diffuse
   samplerSource=SSM_Wrap_WorldGroupSettings
   sortPriority=10
   texture=&T_Asteria_RockCliff_04_D
  }
  n148 = TextureSampleParameter2D(UVs=@n146) {
   group="A_L1 (TD slot 1)"
   parameterName=SpecularMasks
   samplerSource=SSM_Wrap_WorldGroupSettings
   samplerType=SAMPLERTYPE_Masks
   sortPriority=20
   texture=&T_Asteria_RockCliff_04_S
  }
  n149 = TextureSampleParameter2D(UVs=@n146) {
   group="A_L1 (TD slot 1)"
   parameterName=Normals
   samplerSource=SSM_Wrap_WorldGroupSettings
   samplerType=SAMPLERTYPE_Normal
   sortPriority=30
   texture=&T_Asteria_RockCliff_04_N
  }
  n150 = SetMaterialAttributes(BaseColor=@n147.RGB, Opacity=@n147.RGB, Roughness=@n148.B, Specular=@n148.B, Metallic=@n148.B, Normal=@n149.RGB) {attributeSetTypes [
    #1
    #2
    #7
    #8
    #9
    #6
   ]}
  n151 = StaticBoolParameter() {group="A_L1 (TD slot 1)", parameterName=L1 is WS, sortPriority=-1}
  n152 = GetMaterialAttributes(None=StaticSwitch(
    True=StaticSwitchParameter(True=@n143, False=@n138) {group="A_L1 (TD slot 1)", parameterName=L1 WS_Z Only, sortPriority=-1}
    False=@n150
    Value=NamedRerouteDeclaration(None=@n151) {name=L1 is WS, nodeColor=<rgba 0 1 0.0391 1>}
   )) {attributeGetTypes [#6]}
  n153 = ComponentMask(None=@n152.Normal) {g=true, r=true}
  n154 = ScalarParameter() {defaultValue=1.2, group="A_L1 (TD slot 1)", parameterName="L1 Normals Value Baked Prop (non-Nanite)"}
  n155 = ScalarParameter() {defaultValue=2, group="A_L1 (TD slot 1)", parameterName="L1 Normals Value (non-Nanite)"}
  n156 = NamedRerouteUsage()
  n157 = ScalarParameter() {defaultValue=1, group="A_L1 (TD slot 1)", parameterName="L1 Normals Value (Nanite)"}
  n158 = NaniteReplace!(Default=Multiply(A=StaticSwitch(True=@n154, False=@n155, Value=@n156), B=@n157) {constA=2}, Nanite=@n157)
  n159 = Normalize(VectorInput=AppendVector(A=Multiply(A=@n153, B=@n158), B=ComponentMask(None=@n152.Normal) {b=true}))
  n160 = StaticSwitch(
   True=ShadingPathSwitch(
    Default=@n129.MA
    Mobile=MaterialFunctionCall!(Baked Material=@n130, Sedimentary Material=@n131) {materialFunction=/Game/Athena/Helios/Environments/Landscape/Materials/QualityLevelLayers/MF_Helios_CliffSedimentary_Blend_NS}.Result
   )
   False=NamedRerouteDeclaration(None=SetMaterialAttributes!(MaterialAttributes=@n152.MaterialAttributes, Normal=@n159) {attributeSetTypes [#6]}) {name=L1 Baked Rocks, nodeColor=<rgba 0.9219 1 0 1>}
   Value=NamedRerouteDeclaration(None=StaticBoolParameter() {desc=L2 is WS with Baked L1 N, group=1_Rock/Cliff Addons, parameterName=Use Sediment Layers}) {name=Use Sediment, nodeColor=<rgba 0.1953 1 0 1>}
  )
  n161 = GetMaterialAttributes(None=@n160) {attributeGetTypes [
    #1
    #8
    #9
    #6
    #7
   ]}
  NamedRerouteDeclaration(None=@n161.Metallic) {name=L1_Metallic, nodeColor=<rgba 0 0.2031 1 1>}
  n162 = Add(A=NamedRerouteUsage(), B=ScalarParameter() {group=A_L3 Custom Mask, parameterName="• L3_L1 Height_Bias", sortPriority=116})
  n163 = ScalarParameter() {defaultValue=1, group=A_L3 Custom Mask, parameterName="• L3_L1 Height_Contrast", sortPriority=116}
  NamedRerouteDeclaration(None=Saturate(None=Multiply(A=@n162, B=@n163))) {name=L3_M_L1H, nodeColor=<rgba 0 0.5898 1 1>}
  NamedRerouteDeclaration(None=@n161.Normal) {name=L1 Normals, nodeColor=<rgba 0.8984 0 1 1>}
  NamedRerouteDeclaration(None=Constant()) {name=OasisCliffMask, nodeColor=<rgba 0.5 0 1 1>}
  n164 = VertexColor()
  n165 = Constant() {r=2}
  n166 = ScalarParameter() {
   defaultValue=0.5
   desc=Modulates VTX color transition intensity
   group=A_L3 Custom Mask
   parameterName="• L3_VTX Color_Contrast"
   sortPriority=26
  }
  n167 = DotProduct(
   A=Multiply(A=ChannelMaskParameter(None=@n164) {group=A_L3 Custom Mask, parameterName="• L3_VTX Color_Channel", sortPriority=25}.R) {constB=3}
   B=ShadingPathSwitch(Default=@n166, Mobile=Multiply(A=@n166, B=@n165) {desc=compensates for the less dense meshes on mobile creating bigger vertex gradients})
  )
  NamedRerouteDeclaration(None=VertexInterpolator(VS=Saturate(None=@n167)).PS) {name=L3_Vertex Color, nodeColor=<rgba 0 0.5898 1 1>}
  n168 = NamedRerouteUsage()
  n169 = LinearInterpolate(
   A=@n168
   B=ScalarParameter() {defaultValue=0.5, group="A_L1 (TD slot 1)", parameterName=L1 G Mask_Specular Value, sortPriority=21}
   Alpha=NamedRerouteUsage()
  )
  n170 = NamedRerouteUsage()
  NamedRerouteDeclaration(None=StaticSwitch(True=@n168, False=StaticSwitch(True=@n169, False=@n168, Value=@n170), Value=NamedRerouteUsage())) {name=L1 Specular Adjusted, nodeColor=<rgba 0.2422 0 1 1>}
  n171 = SetMaterialAttributes!(
   MaterialAttributes=NamedRerouteUsage()
   BaseColor=NamedRerouteUsage()
   Specular=NamedRerouteUsage()
   Roughness=NamedRerouteUsage()
   Normal=NamedRerouteUsage()
   Metallic=NamedRerouteUsage()
  ) {attributeSetTypes [
    #1
    #8
    #7
    #6
    #9
   ]}
  NamedRerouteDeclaration(None=@n171) {name=Custom Top Layer MA, nodeColor=<rgba 1 0.3044 0 1>}
  n172 = ScalarParameter() {defaultValue=75, group="A_L3_Custom Layer (TD slot 3)", parameterName=L3 WS_Triplanar Fade_Low/Mobile, sortPriority=-1}
  n173 = ScalarParameter() {defaultValue=10, group="A_L3_Custom Layer (TD slot 3)", parameterName=L3 WS_Triplanar Fade, sortPriority=-1}
  n174 = MaterialFunctionCall(
   Texture Scale=ScalarParameter() {defaultValue=2000, group="A_L3_Custom Layer (TD slot 3)", parameterName=L3 WS_Tile, sortPriority=-1}
   Triplanar Fade=MaterialFunctionCall(Low=@n172, High=@n173, Epic=@n173, --------=@n173, Mobile=@n172, Medium=@n173) {materialFunction=&MF_QualitySwitch_Scalar}.Result
  ) {materialFunction=&MF_TriplanarDithered}
  n175 = TextureSampleParameter2D!(UVs=@n174.UVs, "DDX(UVs)"=@n174.DDX, "DDY(UVs)"=@n174.DDY, Apply View MipBias) {
   group="A_L3_Custom Layer (TD slot 3)"
   mipValueMode=TMVM_Derivative
   parameterName=Normals_Texture_3
   samplerSource=SSM_Wrap_WorldGroupSettings
   samplerType=SAMPLERTYPE_Normal
   sortPriority=20
   texture=&FlatNormal
  }
  n176 = Transform(None=MaterialFunctionCall(Tangent Normals=@n175.RGB, Triplanar Mask=@n174.Mask) {materialFunction=&MF_TriplanarNormalTangentToWorld}.World Normals) {transformSourceType=TRANSFORMSOURCE_World, transformType=TRANSFORM_Tangent}
  n177 = ObjectRadius()
  n178 = Multiply(
   A=Multiply(
    A=ScalarParameter() {defaultValue=1, group="A_L3_Custom Layer (TD slot 3)", parameterName=L3 UV Tile, sortPriority=0}
    B=Divide(A=@n177) {constB=256}
   )
   B=TextureCoordinate()
  )
  n179 = TextureSampleParameter2D!(UVs=@n178, Apply View MipBias) {
   group="A_L3_Custom Layer (TD slot 3)"
   parameterName=Normals_Texture_3
   samplerSource=SSM_Wrap_WorldGroupSettings
   samplerType=SAMPLERTYPE_Normal
   sortPriority=20
   texture=&FlatNormal
  }
  n180 = StaticBoolParameter() {group="A_L3_Custom Layer (TD slot 3)", parameterName=L3 Is WS on Mobile, sortPriority=-1}
  n181 = ShadingPathSwitch(Default=@n176, Mobile=StaticSwitch(True=@n176, False=@n179.RGB, Value=@n180))
  n182 = StaticBoolParameter() {
   desc="True - Baked Props Will Use WS projection\r\nFalse - Baked Props Use UV Map"
   group="A_L3_Custom Layer (TD slot 3)"
   parameterName=L3 Is WS
   sortPriority=-2
  }
  n183 = StaticSwitch(True=@n181, False=@n179.RGB, Value=@n182)
  n184 = AppendVector(
   A=Multiply(
    A=ScalarParameter() {defaultValue=1, group="A_L3_Custom Layer (TD slot 3)", parameterName=L3 Normals Value, sortPriority=21}
    B=ComponentMask(None=@n183) {g=true, r=true}
   )
   B=ComponentMask(None=@n183) {b=true}
  )
  n185 = NamedRerouteUsage()
  n186 = Constant3Vector() {constant=<rgba 1 1 0 0>}
  n187 = LinearInterpolate(
   A=@n184
   B=Add(A=@n184, B=Multiply(A=@n185, B=@n186))
   Alpha=Saturate(None=ScalarParameter() {defaultValue=1, group="A_L3_Custom Layer (TD slot 3)", parameterName=L3_Add L2 Normals Value, sortPriority=24})
  )
  NamedRerouteDeclaration(None=Normalize(VectorInput=StaticSwitchParameter(True=@n187, False=@n184) {group="A_L3_Custom Layer (TD slot 3)", parameterName=Add L2 to L3 Normals, sortPriority=23})) {name=CustomTopNormal, nodeColor=<rgba 1 0 0.7734 1>}
  n188 = Multiply(A=NamedRerouteUsage()) {constB=0.009}
  n189 = Divide(A=Add(A=OneMinus(None=@n188)) {constB=-0.5}, B=ComponentMask(None=NamedRerouteUsage()) {g=true})
  n190 = Saturate(None=Add(A=Multiply(A=@n189) {constB=100}) {constB=-30})
  n191 = NamedRerouteUsage()
  n192 = Add(A=Multiply(A=ComponentMask(None=@n191) {r=true}) {constB=20}) {constB=-2}
  NamedRerouteDeclaration(None=StaticSwitchParameter(True=Min(A=@n190, B=@n192), False=Constant()) {parameterName=UseGroundSnowContact}) {name=SnowLandscapeRVTColorMask, nodeColor=<rgba 1 0.7969 0 1>}
  n193 = TextureSampleParameter2D!(UVs=@n174.UVs, "DDX(UVs)"=@n174.DDX, "DDY(UVs)"=@n174.DDY, Apply View MipBias) {
   group="A_L3_Custom Layer (TD slot 3)"
   mipValueMode=TMVM_Derivative
   parameterName=SpecularMasks_3
   samplerSource=SSM_Wrap_WorldGroupSettings
   samplerType=SAMPLERTYPE_Masks
   sortPriority=10
   texture=&T_Fortnite_Default_S
  }
  n194 = TextureSampleParameter2D!(UVs=@n178, Apply View MipBias) {
   group="A_L3_Custom Layer (TD slot 3)"
   parameterName=SpecularMasks_3
   samplerSource=SSM_Wrap_WorldGroupSettings
   samplerType=SAMPLERTYPE_Masks
   sortPriority=10
   texture=&T_Fortnite_Default_S
  }
  n195 = ShadingPathSwitch(Default=@n193.RGB, Mobile=StaticSwitch(True=@n193.RGB, False=@n194.RGB, Value=@n180))
  n196 = MaterialFunctionCall(Float3=StaticSwitch(True=@n195, False=@n194.RGB, Value=@n182)) {materialFunction=&BreakOutFloat3Components}
  NamedRerouteDeclaration(None=@n196.B) {name=CustomTopRoughness, nodeColor=<rgba 1 0.5859 0 1>}
  n197 = Add(A=WorldPosition().Z) {constB=30}
  n198 = ScalarParameter() {
   bUseCustomPrimitiveData=true
   desc="Primitive Data Only, Enable per instance in the details panel."
   group=Waterline
   parameterName=PD_WaterlineHeightOffset
   sortPriority=0
  }
  n199 = NamedRerouteUsage()
  n200 = StaticSwitchParameter(True=OneMinus(None=@n199), False=@n199) {group=A_L4 Landscape Mask, parameterName=L4_L1 SMR_Invert, sortPriority=121}
  n201 = NamedRerouteUsage()
  n202 = StaticSwitchParameter(
   True=StaticSwitchParameter(True=OneMinus(None=@n201), False=@n201) {desc=L3 Mask_World Z Normal, group=A_L4 Landscape Mask, parameterName=L4_World Z Normal_Invert, sortPriority=11}
   False=Constant()
  ) {desc=L3 Mask_World Z Normal, group=A_L4 Landscape Mask, parameterName="== L4_WORLD Z NORMAL MASK ==", sortPriority=10}
  n203 = NamedRerouteUsage()
  n204 = StaticSwitchParameter(True=@n203, False=OneMinus(None=@n203)) {group=A_L4 Landscape Mask, parameterName=L4_VTX Color_Invert, sortPriority=21}
  n205 = Max(A=@n202, B=@n204)
  n206 = Min(A=@n202, B=@n204)
  n207 = StaticSwitchParameter(
   True=StaticSwitchParameter(
    True=@n204
    False=StaticSwitchParameter(True=@n205, False=@n206) {defaultValue=true, group=A_L4 Landscape Mask, parameterName="L4_VTX Color_Blend (T=Max F=Min)", sortPriority=22}
   ) {group=A_L4 Landscape Mask, parameterName=L4_VTX Color_Only, sortPriority=20}
   False=@n202
  ) {group=A_L4 Landscape Mask, parameterName="==L4_VTX COLOR MASK ==", sortPriority=20}
  n208 = NamedRerouteUsage()
  n209 = StaticSwitchParameter(True=OneMinus(None=@n208), False=@n208) {group=A_L4 Landscape Mask, parameterName=L4_AO_Invert, sortPriority=31}
  n210 = StaticSwitchParameter(True=Max(A=@n207, B=@n209), False=Min(A=@n207, B=@n209)) {defaultValue=true, desc=Additive/Subtractive Blending Mode switch, group=A_L4 Landscape Mask, parameterName="L4_AO_Blend (T=Max F=Min)"}
  n211 = StaticSwitchParameter(True=StaticSwitchParameter(True=@n209, False=@n210) {group=A_L4 Landscape Mask, parameterName=L4_AO_Only, sortPriority=30}, False=@n207) {group=A_L4 Landscape Mask, parameterName="== L4_AO MASK ==", sortPriority=30}
  n212 = NamedRerouteUsage()
  n213 = StaticSwitchParameter(True=OneMinus(None=@n212), False=@n212) {group=A_L4 Landscape Mask, parameterName=L4_Curvature_Invert, sortPriority=41}
  n214 = StaticSwitchParameter(True=Max(A=@n211, B=@n213), False=Min(A=@n211, B=@n213)) {
   defaultValue=true
   desc=Additive/Subtractive Blending Mode switch
   group=A_L4 Landscape Mask
   parameterName="CL4_urvature_Blend (T=Max F=Min)"
   sortPriority=42
  }
  n215 = StaticSwitchParameter(True=@n213, False=@n214) {desc=Additive/Subtractive Blending Mode switch, group=A_L4 Landscape Mask, parameterName=L4_Curvature_Only, sortPriority=40}
  n216 = StaticSwitchParameter(True=@n215, False=@n211) {group=A_L4 Landscape Mask, parameterName="== L4_CURVATURE MASK ==", sortPriority=40}
  n217 = NamedRerouteUsage()
  n218 = StaticSwitchParameter(True=OneMinus(None=@n217), False=@n217) {group=A_L4 Landscape Mask, parameterName=L4_Painted_Invert, sortPriority=51}
  n219 = StaticSwitchParameter(True=Max(A=@n216, B=@n218), False=Min(A=@n216, B=@n218)) {
   defaultValue=true
   desc=Additive/Subtractive Blending Mode switch
   group=A_L4 Landscape Mask
   parameterName="L4_Painted_Blend (T=Max F=Min)"
   sortPriority=52
  }
  n220 = StaticSwitchParameter(
   True=StaticSwitchParameter(True=@n218, False=@n219) {desc=Additive/Subtractive Blending Mode switch, group=A_L4 Landscape Mask, parameterName=L4_Painted_Only, sortPriority=50}
   False=@n216
  ) {group=A_L4 Landscape Mask, parameterName="== L4_PAINTED MASK ==", sortPriority=50}
  n221 = NamedRerouteUsage()
  n222 = StaticSwitchParameter(True=@n221, False=OneMinus(None=@n221)) {group=A_L4 Landscape Mask, parameterName=L4_Z Gradient_Invert, sortPriority=61}
  n223 = StaticSwitchParameter(True=Max(A=@n220, B=@n222), False=Min(A=@n220, B=@n222)) {
   defaultValue=true
   desc=Additive/Subtractive Blending Mode switch
   group=A_L4 Landscape Mask
   parameterName="L4_Z Gradient_Blend (T=Max F=Min)"
   sortPriority=62
  }
  n224 = StaticSwitchParameter(True=@n222, False=@n223) {desc=Additive/Subtractive Blending Mode switch, group=A_L4 Landscape Mask, parameterName=L4_Z Gradient_Only, sortPriority=60}
  n225 = StaticSwitchParameter(True=@n224, False=@n220) {group=A_L4 Landscape Mask, parameterName="== L4_Z Gradient Mask ==", sortPriority=60}
  n226 = NamedRerouteUsage()
  n227 = StaticSwitchParameter(True=OneMinus(None=@n226), False=@n226) {group=A_L4 Landscape Mask, parameterName=L4_Waterline_Invert, sortPriority=71}
  n228 = StaticSwitchParameter(True=Max(A=@n225, B=@n227), False=Min(A=@n225, B=@n227)) {
   defaultValue=true
   desc=Additive/Subtractive Blending Mode switch
   group=A_L4 Landscape Mask
   parameterName="L4_Waterline_Blend (T=Max F=Min)"
   sortPriority=72
  }
  n229 = StaticSwitchParameter(True=@n227, False=@n228) {desc=Additive/Subtractive Blending Mode switch, group=A_L4 Landscape Mask, parameterName=L4_Waterline_Only, sortPriority=72}
  n230 = StaticSwitchParameter(True=@n229, False=@n225) {group=A_L4 Landscape Mask, parameterName="== L4_WATERLINE MASK ==", sortPriority=70}
  n231 = NamedRerouteUsage()
  n232 = StaticSwitchParameter(True=OneMinus(None=@n231), False=@n231) {group=A_L4 Landscape Mask, parameterName=L4_SedimentBands_Invert, sortPriority=81}
  n233 = StaticSwitchParameter(True=Max(A=@n230, B=@n232), False=Min(A=@n230, B=@n232)) {
   defaultValue=true
   desc=Additive/Subtractive Blending Mode switch
   group=A_L4 Landscape Mask
   parameterName="L4_SedimentBands_Blend (T=Max F=Min)"
   sortPriority=82
  }
  n234 = StaticSwitchParameter(True=@n232, False=@n233) {desc=Additive/Subtractive Blending Mode switch, group=A_L4 Landscape Mask, parameterName=L4_SedimentBands_Only, sortPriority=80}
  n235 = StaticSwitchParameter(True=@n234, False=@n230) {group=A_L4 Landscape Mask, parameterName="== L4_SEDIMENT BANDS MASK ==", sortPriority=80}
  n236 = NamedRerouteUsage()
  n237 = StaticSwitchParameter(True=OneMinus(None=@n236), False=@n236) {group=A_L4 Landscape Mask, parameterName=L4_WS Noise_Invert, sortPriority=101}
  n238 = StaticSwitchParameter(True=Max(A=@n235, B=@n237), False=Min(A=@n235, B=@n237)) {
   defaultValue=true
   desc=Additive/Subtractive Blending Mode switch
   group=A_L4 Landscape Mask
   parameterName="L4_WS Noise_Blend (T=Max F=Min)"
   sortPriority=102
  }
  n239 = StaticSwitchParameter(True=@n237, False=@n238) {desc=Additive/Subtractive Blending Mode switch, group=A_L4 Landscape Mask, parameterName=L4_WS Noise_Only, sortPriority=100}
  n240 = StaticSwitchParameter(True=@n239, False=@n235) {group=A_L4 Landscape Mask, parameterName="== L4_WS NOISE MASK ==", sortPriority=100}
  n241 = NamedRerouteUsage()
  n242 = StaticSwitchParameter(True=OneMinus(None=@n241), False=@n241) {group=A_L4 Landscape Mask, parameterName=L4_L1 Height_Invert, sortPriority=111}
  n243 = StaticSwitchParameter(True=Max(A=@n240, B=@n242), False=Min(A=@n240, B=@n242)) {
   defaultValue=true
   desc=Additive/Subtractive Blending Mode switch
   group=A_L4 Landscape Mask
   parameterName="L4_L1 Height_Blend (T=Max F=Min)"
   sortPriority=112
  }
  n244 = StaticSwitchParameter(
   True=StaticSwitchParameter(True=@n242, False=@n243) {group=A_L4 Landscape Mask, parameterName=L4_L1 Height_Only, sortPriority=110}
   False=@n240
  ) {group=A_L4 Landscape Mask, parameterName="== L4_L1 Height MASK ==", sortPriority=110}
  n245 = StaticSwitchParameter(True=Max(A=@n244, B=@n200), False=Min(A=@n244, B=@n200)) {
   defaultValue=true
   desc=Additive/Subtractive Blending Mode switch
   group=A_L4 Landscape Mask
   parameterName="L4_L1 SMR_Blend (T=Max F=Min)"
   sortPriority=122
  }
  n246 = StaticSwitchParameter(True=StaticSwitchParameter(True=@n200, False=@n245) {group=A_L4 Landscape Mask, parameterName=L4_L1 SMR_Only, sortPriority=120}, False=@n244) {group=A_L4 Landscape Mask, parameterName="== L4_L1 SMR MASK ==", sortPriority=120}
  n247 = StaticSwitchParameter(True=OneMinus(), False=NamedRerouteUsage()) {group=A_L4 Landscape Mask, parameterName=L4_Height_Invert, sortPriority=591}
  n248 = StaticSwitchParameter(True=Max(A=@n246, B=@n247), False=Min(A=@n246, B=@n247)) {
   defaultValue=true
   desc=Additive/Subtractive Blending Mode switch
   group=A_L4 Landscape Mask
   parameterName="L4_Height_Blend (T=Max F=Min)"
   sortPriority=591
  }
  n249 = StaticSwitchParameter(True=StaticSwitchParameter(True=@n247, False=@n248) {group=A_L4 Landscape Mask, parameterName=L4_Height_Only, sortPriority=590}, False=@n246) {group=A_L4 Landscape Mask, parameterName="== L4_HEIGHT MASK ==", sortPriority=590}
  n250 = StaticSwitchParameter(True=OneMinus(None=@n249), False=@n249) {group=A_L4 Landscape Mask, parameterName=L4 Mask_Invert, sortPriority=0}
  n251 = Add(A=Multiply(A=Add(A=@n197, B=@n198) {constB=30}) {constB=0.015}, B=@n250) {constB=0}
  NamedRerouteDeclaration(None=Saturate(None=Subtract(A=@n251))) {name=Snow_Waterline, nodeColor=<rgba 1 0.8438 0 1>}
  NamedRerouteDeclaration(None=@n116.MudLayer) {name=Landscape_Mud, nodeColor=<rgba 0.4427 0.1763 0.0292 1>}
  n252 = NamedRerouteUsage()
  n253 = NamedRerouteUsage()
  n254 = NamedRerouteUsage()
  n255 = NamedRerouteUsage()
  n256 = StaticBoolParameter() {desc="If set to True, set Custom Top Layer to False", group=A_L4_Landscape, parameterName=Top DesertGrass, sortPriority=-4}
  n257 = StaticBoolParameter() {desc="If set to True, set Custom Top Layer to False", group=A_L4_Landscape, parameterName=Top ChaparralGrass, sortPriority=-4}
  n258 = StaticSwitch(
   True=NamedRerouteUsage()
   False=StaticSwitch(True=NamedRerouteUsage(), False=StaticSwitch(True=@n254, False=@n255, Value=@n256), Value=@n257)
   Value=StaticBoolParameter() {group=A_L4_Landscape, parameterName=TopGravel, sortPriority=-4}
  )
  n259 = StaticBoolParameter() {group=A_L4_Landscape, parameterName=TopMud, sortPriority=-4}
  n260 = StaticSwitch(
   True=NamedRerouteUsage()
   False=StaticSwitch(
    True=NamedRerouteUsage()
    False=StaticSwitch(True=@n253, False=@n258, Value=@n259)
    Value=StaticBoolParameter() {group=A_L4_Landscape, parameterName=TopSand, sortPriority=-4}
   )
   Value=StaticBoolParameter() {group=A_L4_Landscape, parameterName=TopAsphalt, sortPriority=-4}
  )
  n261 = StaticBoolParameter() {group=A_L4_Landscape, parameterName=TopScree, sortPriority=-4}
  n262 = StaticBoolParameter() {desc="If set to True, set TopLayer to False", group=A_L4_Landscape, parameterName=TopSnow, sortPriority=-4}
  n263 = GetMaterialAttributes(None=StaticSwitch(
    True=NamedRerouteUsage()
    False=StaticSwitch(True=@n252, False=@n260, Value=@n261)
    Value=NamedRerouteDeclaration(None=@n262) {name=UseTopSnow, nodeColor=<rgba 0.3594 0 1 1>}
   )) {attributeGetTypes [#10]}
  n264 = GetMaterialAttributes(None=@n263.MaterialAttributes) {attributeGetTypes [#6]}
  n265 = NamedRerouteUsage()
  n266 = Add(A=@n264.Normal, B=Multiply(A=Transform(None=@n265), B=Constant3Vector() {constant=<rgba 1 1 0 0>}))
  n267 = Saturate(None=ScalarParameter() {defaultValue=0.5, group=A_L4_Landscape, parameterName=L4_Add L2 Normals Value, sortPriority=24})
  n268 = StaticSwitchParameter(True=Normalize(VectorInput=LinearInterpolate(A=@n264.Normal, B=@n266, Alpha=@n267)), False=@n264.Normal) {group=A_L4_Landscape, parameterName=Add L2 to L4 Normals, sortPriority=23}
  n269 = NamedRerouteUsage()
  NamedRerouteDeclaration(None=SetMaterialAttributes!(MaterialAttributes=@n264.MaterialAttributes, Normal=StaticSwitch(True=@n268, False=@n264.Normal, Value=@n269)) {attributeSetTypes [#6]}) {name=LandscapeLayerMaterial, nodeColor=<rgba 0.5 0 1 1>}
  n270 = ScalarParameter() {defaultValue=3.2, group=A_L4 Mask, parameterName="• L4 Height Mask_Contrast", sortPriority=595}
  n271 = Subtract(B=@n270) {constA=0, desc=Low negative value}
  n272 = Add(A=@n270) {desc=High Positive}
  n273 = Add(A=NamedRerouteUsage(), B=ScalarParameter() {defaultValue=-0.5, group=A_L4 Mask, parameterName="• L4 Height Mask_Bias", sortPriority=595})
  NamedRerouteDeclaration(None=Saturate(None=LinearInterpolate(A=@n271, B=@n272, Alpha=@n273) {desc=Lerps between greater range of value})) {name=LandscapeLayerHeight, nodeColor=<rgba 0.5 0 1 1>}
  NamedRerouteDeclaration(None=@n196.G) {name=CustomTopMetallic, nodeColor=<rgba 1 0.9844 0 1>}
  NamedRerouteDeclaration(None=@n39.RVT Color) {name=RVT Color, nodeColor=<rgba 0 1 0.2734 1>}
  NamedRerouteDeclaration(None=@n160) {name=MobileSedimentMaterial, nodeColor=<rgba 0.6875 1 0 1>}
  NamedRerouteDeclaration(None=@n196.R) {name=CustomTopSpecular, nodeColor=<rgba 1 0.0234 0 1>}
  n274 = TextureSampleParameter2D!(UVs=@n174.UVs, "DDX(UVs)"=@n174.DDX, "DDY(UVs)"=@n174.DDY, Apply View MipBias) {
   group="A_L3_Custom Layer (TD slot 3)"
   mipValueMode=TMVM_Derivative
   parameterName=Diffuse_Texture_3
   samplerSource=SSM_Wrap_WorldGroupSettings
   sortPriority=1
   texture=/Game/Global/Textures/Default/Blanks/T_MidgreysRGB
  }
  n275 = TextureSampleParameter2D!(UVs=@n178, Apply View MipBias) {
   group="A_L3_Custom Layer (TD slot 3)"
   parameterName=Diffuse_Texture_3
   samplerSource=SSM_Wrap_WorldGroupSettings
   sortPriority=1
   texture=&T_MidgreysRGB
  }
  n276 = ShadingPathSwitch(Default=@n274.RGBA, Mobile=StaticSwitch(True=@n274.RGBA, False=@n275.RGBA, Value=@n180))
  n277 = StaticSwitch(True=@n276, False=@n275.RGBA, Value=@n182)
  NamedRerouteDeclaration(None=ComponentMask(None=@n277) {b=true, g=true, r=true}) {name=CustomTopBaseColor, nodeColor=<rgba 0 0.2969 1 1>}
  n278 = ScalarParameter() {defaultValue=1, desc=Set to Zero if no AO texture is desired, group="A_L2 (TD slot 2)", parameterName=UseBakedAOTexture, sortPriority=520}
  n279 = StaticBoolParameter() {
   desc="OFF= Nanite uses VTX and Default uses Textures (Prop Rocks are the only meshes  that gets tesselated and displaced on nanite platforms, by default thay won`t use VTX)\r\nON= Forte VTX for all (VTX masks are only used on big custom sculpts that have a custom nanite mesh with no UVs)\r\n"
   group="A_L2 (TD slot 2)"
   parameterName=Use VTX Curvature/AO/Painted on Nanite
   sortPriority=501
  }
  n280 = ScalarParameter() {defaultValue=1, group="A_L2 (TD slot 2)", parameterName="UV 0 or 1? (SpecularMask_2)", sliderMax=1, sortPriority=1}
  n281 = TextureSampleParameter2D(UVs=LinearInterpolate(A=TextureCoordinate(), B=TextureCoordinate() {coordinateIndex=1}, Alpha=Saturate(None=@n280))) {
   desc="Actually Just a Mask texture, named Spec_2 for TD to work"
   group="A_L2 (TD slot 2)"
   parameterName=SpecularMasks_2
   samplerSource=SSM_Wrap_WorldGroupSettings
   samplerType=SAMPLERTYPE_Masks
   sortPriority=500
   texture=&T_Fortnite_Default_S
  }
  n282 = TextureSampleParameter2D() {
   group="A_L1 (TD slot 1)"
   parameterName=SpecularMasks
   samplerSource=SSM_Wrap_WorldGroupSettings
   samplerType=SAMPLERTYPE_Masks
   sortPriority=20
   texture=&T_Asteria_RockCliff_04_S
  }
  n283 = MaterialFunctionCall(
   R=StaticSwitch(True=@n281.R, False=@n282.RGB, Value=NamedRerouteUsage())
   G=StaticSwitch(True=@n281.G, False=Constant(), Value=NamedRerouteUsage())
   B=@n281.B
  ) {materialFunction=&AppendMany}
  n284 = NaniteReplace(Default=@n283.RGB, Nanite=StaticSwitch(True=VertexColor(), False=@n283.RGB, Value=@n279))
  n285 = LinearInterpolate(
   A=Constant() {r=1}
   B=NamedRerouteDeclaration(None=ChannelMaskParameter(None=@n284) {group="A_L2 (TD slot 2)", parameterName=AO_Channel, sortPriority=520}.R) {name=L2_AO_Baked, nodeColor=<rgba 0 1 0 1>}
   Alpha=Saturate(None=@n278)
  )
  n286 = DistanceFieldApproxAO!(
   World Position
   Normal
   BaseDistance=ScalarParameter() {defaultValue=55, desc=55, group="A_L2 (TD slot 2)", parameterName=RVT_AO_DF Distance, sortPriority=520}
   Radius=ScalarParameter() {defaultValue=150, desc=150, group="A_L2 (TD slot 2)", parameterName=RVT_AO_DF Radius, sortPriority=520}
  ) {numSteps=3}
  n287 = Divide(A=Subtract(A=PixelDepth().R, B=Constant() {r=5000}), B=Constant() {r=10000})
  n288 = Saturate(None=Subtract(A=OneMinus(None=@n286), B=Saturate(None=@n287)))
  n289 = Multiply(
   A=Power(Base=OneMinus(None=@n288), Exp=ScalarParameter() {defaultValue=1, group="A_L2 (TD slot 2)", parameterName=DFAO_Power, sortPriority=51})
   B=ScalarParameter() {defaultValue=1, group="A_L2 (TD slot 2)", parameterName=DFAO_Mult, sortPriority=51}
  )
  n290 = ShadingPathSwitch(Default=QualitySwitch(Default=@n285, Epic=Saturate(None=@n289)), Mobile=@n285)
  NamedRerouteDeclaration(None=@n290) {name=AO_DF, nodeColor=<rgba 1 0 0.3516 1>}
  n291 = MaterialFunctionCall() {materialFunction=&ObjectLocalBounds}
  n292 = Subtract(A=LocalPosition().XYZ, B=@n291.Local Bounds Minimum)
  NamedRerouteDeclaration(None=ComponentMask(None=Divide(A=@n292, B=@n291.Local Bounds Size)) {b=true}) {name=L2_VerticalGradient, nodeColor=<rgba 0.1719 1 0 1>}
  NamedRerouteDeclaration(None=@n161.Specular) {name=L1_Specular, nodeColor=<rgba 0.6641 1 0 1>}
  n293 = StaticBoolParameter() {
   desc="L1 Specular G channel wll be used as Metallic Mask.\r\nThis toggle overrides other L1 Specular Value Adjustments"
   group="A_L1_Rock Surface (TD slot 1)"
   parameterName=G Channel IS Metal
   sortPriority=11
  }
  NamedRerouteDeclaration(None=StaticSwitch(
    True=NamedRerouteUsage()
    False=Constant()
    Value=NamedRerouteDeclaration(None=@n293) {name=Spec_G Channel IS Metal, nodeColor=<rgba 1 0.7734 0 1>}
   )) {name=L1 Metallic Adjusted, nodeColor=<rgba 1 0 0.9844 1>}
  n294 = StaticBoolParameter() {
   defaultValue=true
   desc="Used for cliff Moss in CH5)"
   group="A_L1 (TD slot 1)"
   parameterName="L1 G Mask_Tint && Spec Adjustments"
   sortPriority=15
  }
  n295 = NamedRerouteUsage()
  n296 = ScalarParameter() {group="A_L1 (TD slot 1)", parameterName=L1 G Mask_Desaturation, sortPriority=16}
  n297 = VectorParameter() {
   group="A_L1 (TD slot 1)"
   parameterName=L1 G Mask_Tint
   sortPriority=16
   channelNames {a="0=Multiply 1=Solid Color", b="", g="", r=""}
   defaultValue=<rgba 1 1 1 0>
  }
  n298 = LinearInterpolate(A=Multiply(A=Desaturation(None=@n295, Fraction=@n296), B=@n297), B=@n297, Alpha=@n297)
  n299 = NamedRerouteUsage()
  n300 = NamedRerouteUsage()
  n301 = Desaturation(None=@n300, Fraction=ScalarParameter() {group="A_L1 (TD slot 1)", parameterName=L1 Color_Desaturation, sortPriority=11})
  n302 = VectorParameter() {
   group="A_L1 (TD slot 1)"
   parameterName=L1 Color_Tint
   sortPriority=11
   channelNames {a="0=Multiply 1=Solid Color", b="", g="", r=""}
   defaultValue=<rgba 1 1 1 0>
  }
  n303 = StaticSwitch(
   True=LinearInterpolate(A=Multiply(A=@n301, B=@n302), B=@n302, Alpha=@n302)
   False=@n300
   Value=StaticBoolParameter() {defaultValue=true, group="A_L1 (TD slot 1)", parameterName=L1 Color_Tint, sortPriority=10}
  )
  NamedRerouteDeclaration(None=StaticSwitch(
    True=LinearInterpolate(A=@n303, B=@n298, Alpha=@n299)
    False=@n303
    Value=NamedRerouteDeclaration(None=@n294) {name=Use L1 G Mask_Adjustments, nodeColor=<rgba 1 0 0.9844 1>}
   )) {name=L1 Base Color Adjusted, nodeColor=<rgba 0 1 0.6953 1>}
  NamedRerouteDeclaration(None=@n161.BaseColor) {name=L1_BaseColor, nodeColor=<rgba 1 0.375 0 1>}
  n304 = Saturate(None=ComponentMask(None=VertexNormalWS()) {b=true})
  n305 = GetMaterialAttributes(None=NamedRerouteUsage()) {attributeGetTypes [#6]}
  n306 = Saturate(None=ComponentMask(None=Transform(None=@n305.Normal)) {b=true})
  n307 = Add(
   A=StaticSwitchParameter(True=@n304, False=@n306) {desc="T= VTX only\r\nF= L2 Normals", group=A_L3 Custom Mask, parameterName=L3_World Z Normal_Use VTX, sortPriority=15}
   B=ScalarParameter() {defaultValue=-0.2, group=A_L3 Custom Mask, parameterName="• L3_World Z Normal_Bias", sortPriority=15}
  )
  n308 = ScalarParameter() {defaultValue=1.5, group=A_L3 Custom Mask, parameterName="• L3_World Z Normal_Scale", sortPriority=15}
  NamedRerouteDeclaration(None=Saturate(None=Multiply(A=@n307, B=@n308))) {name=L3_Normal Z, nodeColor=<rgba 0 0.5898 1 1>}
  NamedRerouteDeclaration(None=@n88) {name=L3 Gradient Final, nodeColor=<rgba 1 0.8672 0 1>}
  n309 = VectorParameter() {
   group=Cliff
   parameterName="Sediment_a | Pos && Scale"
   sortPriority=306
   channelNames {a=Scale, b=Z, g=Y, r=X}
   defaultValue=<rgba 0 0 0 2500>
  }
  n310 = VectorParameter() {
   group=Cliff
   parameterName="Sediment_a | Wave Dist Pos && Freq"
   sortPriority=307
   channelNames {a=Freq, b=Z, g=Y, r=X}
   defaultValue=<rgba 0 0 0 10000>
  }
  n311 = MaterialFunctionCall!(
   WorldPosition
   Sediment L2_WS | Position Offset_1=@n309
   Sediment L2_WS | Scale=@n309
   Use Wavy Distortion=StaticBoolParameter() {group=Cliff, parameterName=Sediment_a | Use Wave Distortion, sortPriority=306}
   Wave Distortion Intensity=ScalarParameter() {defaultValue=5000, group=Cliff, parameterName=Sediment_a | Wave Dist Intensity, sortPriority=307}
   Sediment L2_WS | Wave DistPos Offset=@n310
   Wave Distortion Freq=@n310
  ) {materialFunction=/Game/Global/Base4Layers_2022/MaterialFunctions/MF_Cliffs_SedimentWS_UVs}
  n312 = TextureObjectParameter() {group="A_L2 (TD slot 2)", parameterName=Diffuse_Texture_2, texture=&T_Asteria_RockCliff_04_D}
  n313 = TextureObjectParameter() {
   desc="Actually Just a Mask texture, named Spec_2 for TD to work"
   group="A_L2 (TD slot 2)"
   parameterName=SpecularMasks_2
   samplerType=SAMPLERTYPE_Masks
   sortPriority=500
   texture=&T_Fortnite_Default_S
  }
  n314 = TextureObjectParameter() {group="A_L2 (TD slot 2)", parameterName=Normals_Texture_2, samplerType=SAMPLERTYPE_Normal, texture=&FlatNormal}
  n315 = ScalarParameter() {defaultValue=0.2, desc=XY Projection Transition Contrast, group=Cliff, parameterName=Sediment_a | PTC, sortPriority=306}
  n316 = MaterialFunctionCall!(UV_XZ=@n311.XZ, UV_YZ=@n311.XZ, T_Diffuse=@n312, T_Specular=@n313, T_Normal=@n314, PTC=@n315) {materialFunction=/Game/Global/Base4Layers_2022/MaterialFunctions/MF_Cliffs_SedimentWS}
  n317 = NamedRerouteUsage()
  n318 = StaticBool() {value=true}
  NamedRerouteDeclaration(None=ShadingPathSwitch(
    Default=@n316.SetMaterialAttributes
    Mobile=MaterialFunctionCall!(OasisMask=@n317, VariationMasks Mobile, IsStaticMesh=@n318) {materialFunction=/Hera/Environment/Terrain/Landscape/Materials/Layers/QualityLevel/MF_Hera_CliffSedimentary_Mobile}.Result
   )) {name=Sediment MA, nodeColor=<rgba 0.6172 0 1 1>}
  NamedRerouteDeclaration(None=@n316.Sediment Normal B Mask) {name=Sediment Normal B mask, nodeColor=<rgba 1 0 0.8906 1>}
  NamedRerouteDeclaration(None=@n311.Sediment Wave Distortion) {name=Sediment Wave Distortion, nodeColor=<rgba 0 1 0.6484 1>}
  NamedRerouteDeclaration(None=ShadingPathSwitch(Default=@n40.Waterline Height Mask, Mobile=@n41.Waterline Height Mask)) {name=Waterline Gradient, nodeColor=<rgba 0.2422 1 0 1>}
  n319 = StaticSwitch(
   True=NamedRerouteUsage()
   False=TextureSample(UVs=NamedRerouteUsage()) {texture=/Game/FN_PublicTurntable/Turntable_Lowpoly/Content/BlueprintAssets/Checker}.RGB
   Value=NamedRerouteUsage()
  )
  n320 = SetMaterialAttributes!(MaterialAttributes=@n161.MaterialAttributes, BaseColor=NamedRerouteUsage(), Specular=NamedRerouteUsage(), Metallic=NamedRerouteUsage()) {attributeSetTypes [#1, #8, #9]}
  n321 = GetMaterialAttributes(None=SetMaterialAttributes!(
    MaterialAttributes=@n320
    BaseColor=StaticSwitchParameter(True=@n319, False=@n320) {group="A_L1 (TD slot 1)", parameterName=Debug L1 UV's, sortPriority=-100}
   ) {attributeSetTypes [#1]}) {attributeGetTypes [
    #1
    #8
    #7
    #6
    #2
   ]}
  NamedRerouteDeclaration(None=@n321.Opacity) {name=L1_Height_1, nodeColor=<rgba 0.9922 0 1 1>}
  n322 = ScalarParameter() {desc="1= object top bound\r\n0= object bottom bound", group=A_L3 Custom Mask, parameterName="• L3_Z Gradient_HBottom", sortPriority=63}
  n323 = ScalarParameter() {
   defaultValue=1
   desc="1= object top bound\r\n0= object bottom bound"
   group=A_L3 Custom Mask
   parameterName="• L3_Z Gradient_HTop"
   sortPriority=62
  }
  NamedRerouteDeclaration(None=SmoothStep(Min=@n322, Max=@n323, Value=NamedRerouteUsage())) {name=L3_Z Gradient, nodeColor=<rgba 0 0.5898 1 1>}
  n324 = StaticSwitchParameter(True=NamedRerouteUsage(), False=NamedRerouteUsage()) {
   desc="** DF AO is for Epic Settings Only!True= AO Baked from VTX or Texture\r\nFalse= AO Distance Field"
   group=A_L3 Custom Mask
   parameterName=L3_AO_Force Baked on Nanite
   sortPriority=35
  }
  n325 = Add(A=@n324, B=ScalarParameter() {group=A_L3 Custom Mask, parameterName="• L3_AO_Bias", sortPriority=36})
  n326 = ScalarParameter() {defaultValue=1, group=A_L3 Custom Mask, parameterName="• L3_AO_Contrast", sortPriority=36}
  NamedRerouteDeclaration(None=Saturate(None=Multiply(A=@n325, B=@n326))) {name=L3_AO Mask, nodeColor=<rgba 0 0.5898 1 1>}
  n327 = TextureSample!(UVs=@n134.UVs, Tex, "DDX(UVs)"=@n134.DDX, "DDY(UVs)"=@n134.DDY, Apply View MipBias) {
   mipValueMode=TMVM_Derivative
   samplerSource=SSM_Wrap_WorldGroupSettings
   texture=&Checker
  }
  NamedRerouteDeclaration(None=@n327.RGB) {name=L1 WS_UV Debug, nodeColor=<rgba 0 1 0.4844 1>}
  n328 = NamedRerouteUsage()
  n329 = NamedRerouteUsage()
  n330 = Constant() {r=10}
  n331 = Constant() {r=75}
  n332 = MaterialFunctionCall(
   Texture Scale=ScalarParameter() {defaultValue=700, group="A_L2 (TD slot 2)", parameterName=Cliffs WS Detail__Scale, sortPriority=1}
   Triplanar Fade=MaterialFunctionCall(Low=@n331, High=@n330, Epic=@n330, --------=@n330, Mobile=@n331, Medium=@n330) {materialFunction=&MF_QualitySwitch_Scalar}.Result
  ) {materialFunction=&MF_TriplanarDithered}
  n333 = TextureSampleParameter2D!(UVs=@n332.UVs, "DDX(UVs)"=@n332.UVs, "DDY(UVs)"=@n332.UVs, Apply View MipBias) {
   group="A_L2 (TD slot 2)"
   mipValueMode=TMVM_Derivative
   parameterName=Cliffs WS Detail_Texture
   samplerSource=SSM_Wrap_WorldGroupSettings
   sortPriority=0
   texture=/Game/Environments/Apollo/Rocks/Natural_Rocks/Textures/T_RockDetails_02D
  }
  n334 = SquareRoot(None=@n333.RGB)
  n335 = ScalarParameter() {defaultValue=1, group="A_L2 (TD slot 2)", parameterName=Cliffs WS Detail__Color, sortPriority=2}
  n336 = Multiply(A=Multiply(A=NamedRerouteUsage(), B=LinearInterpolate(B=@n334, Alpha=@n335) {constA=1}), B=Constant() {r=1.5})
  n337 = Constant() {r=-0.25}
  n338 = Constant() {r=1.5}
  n339 = NamedRerouteUsage()
  n340 = Clamp(None=Saturate(None=LinearInterpolate(A=@n337, B=@n338, Alpha=@n339)), Max=Constant() {r=0.5}) {clampMode=CMODE_ClampMax}
  n341 = Add(A=Constant() {r=-1.1}, B=Multiply(A=Constant() {r=2.05}, B=@n333.R))
  n342 = LinearInterpolate(A=Constant() {r=7}, B=Constant() {r=-0.8}, Alpha=@n333.R)
  n343 = ScalarParameter() {defaultValue=1, group="A_L2 (TD slot 2)", parameterName=Cliffs WS Detail__Spec, sortPriority=3}
  n344 = LinearInterpolate(B=Multiply(A=Saturate(None=@n341), B=Saturate(None=@n342)), Alpha=@n343) {constA=1}
  n345 = Multiply(B=@n343) {constA=0.25}
  n346 = ShadingPathSwitch(
   Default=SetMaterialAttributes!(MaterialAttributes=@n329, BaseColor=Saturate(None=@n336), Specular=Multiply(A=@n340, B=@n344)) {attributeSetTypes [#1, #8]}
   Mobile=SetMaterialAttributes!(MaterialAttributes=@n329, Specular=Multiply(A=@n339, B=@n345)) {attributeSetTypes [#8]}
  )
  n347 = ScalarParameter() {defaultValue=10, group="A_L2 (TD slot 2)", parameterName=L2 WS Normals_Triplanar Fade, sortPriority=101}
  n348 = ScalarParameter() {defaultValue=75, group="A_L2 (TD slot 2)", parameterName=L2 WS Normals_Triplanar Fade_Low/Mobile, sortPriority=101}
  n349 = MaterialFunctionCall(
   Texture Scale=ScalarParameter() {defaultValue=700, group="A_L2 (TD slot 2)", parameterName=L2 WS Normals_Tile, sortPriority=101}
   Triplanar Fade=MaterialFunctionCall(Low=@n348, High=@n347, Epic=@n347, --------=@n347, Mobile=@n348, Medium=@n347) {materialFunction=&MF_QualitySwitch_Scalar}.Result
  ) {materialFunction=&MF_TriplanarDithered}
  n350 = TextureSampleParameter2D(UVs=@n349.UVs, "DDX(UVs)"=@n349.UVs, "DDY(UVs)"=@n349.UVs) {
   group="A_L2 (TD slot 2)"
   mipValueMode=TMVM_Derivative
   parameterName=Normals_Texture_2
   samplerSource=SSM_Wrap_WorldGroupSettings
   samplerType=SAMPLERTYPE_Normal
   texture=&FlatNormal
  }
  n351 = MaterialFunctionCall(Tangent Normals=@n350.RGB, Triplanar Mask=@n349.Mask) {materialFunction=&MF_TriplanarNormalTangentToWorld}
  n352 = TextureCoordinate()
  n353 = Multiply(
   A=Divide(A=ObjectRadius()) {constB=256}
   B=ScalarParameter() {defaultValue=1, group="A_L2 (TD slot 2)", parameterName=L2 Cliff Normal Scale, sortPriority=101}
  )
  n354 = TextureSampleParameter2D!(UVs=NamedRerouteDeclaration(None=Multiply(A=@n352, B=@n353)) {name=L2 Cliff Uvs, nodeColor=<rgba 0 1 0.9531 1>}, Apply View MipBias) {
   group="A_L2 (TD slot 2)"
   parameterName=Normals_Texture_2
   samplerSource=SSM_Wrap_WorldGroupSettings
   samplerType=SAMPLERTYPE_Normal
   texture=&FlatNormal
  }
  n355 = Multiply(
   A=StaticSwitchParameter(True=Transform(None=@n351.World Normals) {transformSourceType=TRANSFORMSOURCE_World, transformType=TRANSFORM_Tangent}, False=@n354.RGB) {group="A_L2 (TD slot 2)", parameterName=L2_WS Detail Normals, sortPriority=33}
   B=Constant3Vector() {constant=<rgba 1 1 0 0>}
  )
  n356 = ScalarParameter() {defaultValue=1, group="A_L2 (TD slot 2)", parameterName=L2 Cliff Normal Value, sortPriority=101}
  n357 = NamedRerouteUsage()
  n358 = Normalize(VectorInput=Add(A=@n357, B=Multiply(A=@n355, B=@n356)))
  n359 = SetMaterialAttributes!(
   MaterialAttributes=@n346
   Normal=StaticSwitch(
    True=@n357
    False=StaticSwitchParameter(True=@n358, False=@n357) {group="A_L2 (TD slot 2)", parameterName=L2 Cliff Normal Details, sortPriority=29}
    Value=NamedRerouteUsage()
   )
  ) {attributeSetTypes [#6]}
  n360 = NamedRerouteUsage()
  n361 = NamedRerouteUsage()
  n362 = Add(B=@n361) {constA=0.5}
  n363 = NamedRerouteUsage()
  n364 = Constant() {r=0.5}
  n365 = Constant() {r=1.25}
  n366 = NamedRerouteUsage()
  n367 = OneMinus(None=Multiply(A=LinearInterpolate(A=@n364, B=@n365, Alpha=@n366), B=NamedRerouteUsage()))
  n368 = Saturate(None=LinearInterpolate(
    A=ScalarParameter() {defaultValue=-3.75, group="A_L2 (TD slot 2)", parameterName=L2BakedPropBottomTrick_LerpA, sortPriority=901}
    B=ScalarParameter() {defaultValue=3.75, group="A_L2 (TD slot 2)", parameterName=L2BakedPropBottomTrick_LerpB, sortPriority=902}
    Alpha=Subtract(A=@n363, B=@n367)
   ))
  n369 = Saturate(None=LinearInterpolate(A=Multiply(A=@n362, B=@n361), B=@n361, Alpha=@n368))
  n370 = NamedRerouteUsage()
  n371 = Saturate(None=LinearInterpolate(A=Multiply(A=@n370) {constB=2}, B=@n370, Alpha=@n368)) {desc="This should be clamped like this, but it previously wasn't"}
  n372 = TextureCoordinate() {coordinateIndex=1}
  n373 = TextureSampleParameter2D(UVs=@n372) {
   group="A_L2 (TD slot 2)"
   parameterName=Normals_Texture_2
   samplerSource=SSM_Wrap_WorldGroupSettings
   samplerType=SAMPLERTYPE_Normal
   texture=&FlatNormal
  }
  n374 = Transform(None=ShaderStageSwitch(
    PixelShader=MaterialFunctionCall!(UVs=@n372, Normal=@n373.RGB) {materialFunction=/Engine/Functions/Engine_MaterialFunctions01/Texturing/DeriveTangentBasis}.World Space Normal
    VertexShader=@n373.RGB
   )) {transformSourceType=TRANSFORMSOURCE_World, transformType=TRANSFORM_Tangent}
  n375 = StaticSwitch(
   True=ScalarParameter() {defaultValue=1.2, group="A_L2 (TD slot 2)", parameterName="L2 Baked Normals Value Baked Prop (non-Nanite)", sortPriority=50}
   False=ScalarParameter() {defaultValue=2, group="A_L2 (TD slot 2)", parameterName="L2 Baked Normals Value (non-Nanite)", sortPriority=50}
   Value=NamedRerouteUsage()
  )
  n376 = ScalarParameter() {defaultValue=1, group="A_L2 (TD slot 2)", parameterName="L2 Baked Normals Value (Nanite)", sortPriority=50}
  n377 = Multiply(A=ComponentMask(None=@n374) {g=true, r=true}, B=NaniteReplace(Default=Multiply(A=@n375, B=@n376) {constA=2}, Nanite=@n376))
  n378 = ComponentMask(None=@n374) {b=true}
  n379 = NamedRerouteUsage()
  n380 = Constant3Vector() {constant=<rgba 1 1 0 0>}
  n381 = Normalize(VectorInput=Add(A=AppendVector(A=@n377, B=@n378), B=Multiply(A=@n379, B=@n380)))
  n382 = StaticSwitchParameter(
   True=StaticSwitch(True=NamedRerouteUsage(), False=ShadingPathSwitch(Default=@n381, Mobile=@n379), Value=NamedRerouteUsage())
   False=NamedRerouteUsage()
  ) {
   defaultValue=true
   desc=True uses UV1 for the normals. False uses L1 Normals
   group="A_L2 (TD slot 2)"
   parameterName=L2_Use Detail Normals
   sortPriority=33
  }
  n383 = NamedRerouteUsage()
  n384 = NamedRerouteUsage()
  n385 = StaticSwitch(
   True=StaticSwitch(
    True=SetMaterialAttributes!(MaterialAttributes=@n360, BaseColor=@n369, Roughness=@n371, Normal=@n382) {attributeSetTypes [#1, #7, #6]}
    False=SetMaterialAttributes(MaterialAttributes=@n360, BaseColor=@n383, Roughness=@n384, Normal=@n382) {attributeSetTypes [#1, #7, #6]}
    Value=StaticBoolParameter() {defaultValue=true, group="A_L2 (TD slot 2)", parameterName=UseL2BottomTrick, sortPriority=900}
   )
   False=MaterialFunctionCall!(
    UV's=NamedRerouteUsage()
    Checker true=StaticBoolParameter() {group="A_L2 (TD slot 2)", parameterName="[DEBUG] L2 Cliff UVs", sortPriority=-10}
    Material In=QualitySwitch(Default=@n346, Low=@n346, High=@n346, Medium=@n346, Epic=@n359)
   ) {materialFunction=/Game/Global/Base4Layers_2022/MaterialFunctions/MF_Checker}.Switch
   Value=NamedRerouteUsage()
  )
  n386 = GetMaterialAttributes(None=StaticSwitch(True=ShadingPathSwitch(Default=@n385, Mobile=@n328), False=@n385, Value=NamedRerouteUsage())) {attributeGetTypes [#10, #6]}
  NamedRerouteDeclaration(None=@n386.MaterialAttributes) {name=L2_Material, nodeColor=<rgba 0 1 0.3839 1>}
  n387 = ShadingPathSwitch(Default=Max(A=@n250, B=NamedRerouteUsage()), Mobile=@n250)
  n388 = NamedRerouteUsage()
  n389 = Multiply(A=StaticSwitch(True=Min(A=@n387, B=@n388), False=@n250, Value=NamedRerouteUsage())) {constB=2, desc=2}
  n390 = Subtract(A=LinearInterpolate(
    A=ScalarParameter() {group=A_L4_Landscape, parameterName=Landscape H Min}
    B=ScalarParameter() {defaultValue=1, group=A_L4_Landscape, parameterName=Landscape H Max}
    Alpha=NamedRerouteUsage()
   )) {desc=1}
  NamedRerouteDeclaration(None=Saturate(None=Add(A=@n389, B=@n390))) {name=L4 Mask Blended, nodeColor=<rgba 0.5 0 1 1>}
  NamedRerouteDeclaration(None=@n129.Sediment Bands Mask) {name=Sediment Bands Mask, nodeColor=<rgba 0.3073 1 0.5732 1>}
  n391 = Add(A=NamedRerouteUsage(), B=ScalarParameter() {group=A_L3 Custom Mask, parameterName="• L3_Curvature_Bias", sortPriority=46})
  n392 = ScalarParameter() {defaultValue=1, group=A_L3 Custom Mask, parameterName="• L3_Curvature_Scale", sortPriority=46}
  NamedRerouteDeclaration(None=Saturate(None=Multiply(A=@n391, B=@n392))) {name=L3_Curvature/Thickness, nodeColor=<rgba 0 0.5898 1 1>}
  NamedRerouteDeclaration(None=SmoothStep(
    Min=ScalarParameter() {group=A_L3 Custom Mask, parameterName="• Waterline_Min", sortPriority=75}
    Max=ScalarParameter() {defaultValue=0.5, group=A_L3 Custom Mask, parameterName="• Waterline_Max", sortPriority=75}
    Value=NamedRerouteUsage()
   )) {name=L3_Waterline, nodeColor=<rgba 0 0.5898 1 1>}
  n393 = Constant() {r=30}
  n394 = Constant() {r=50}
  n395 = MaterialFunctionCall(
   Texture Scale=ScalarParameter() {defaultValue=2000, group=A_L3 Custom Mask, parameterName="• L3_WS Noise Tile", sortPriority=106}
   Triplanar Fade=MaterialFunctionCall(Low=@n394, High=@n393, Epic=@n393, --------=@n393, Mobile=@n394, Medium=@n393) {materialFunction=&MF_QualitySwitch_Scalar}.Result
  ) {materialFunction=&MF_TriplanarDithered}
  n396 = TextureSampleParameter2D!(UVs=@n395.UVs, "DDX(UVs)"=@n395.UVs, "DDY(UVs)"=@n395.UVs, Apply View MipBias) {
   group=A_L3 Custom Mask
   mipValueMode=TMVM_Derivative
   parameterName="• L3_WS Noise_WorldSpaceNoise"
   sortPriority=105
   texture=/Game/Environments/Asteria/Props/LobbyIslandSpace/FogSheet/Textures/TilingNoise
  }
  n397 = Add(A=@n396.R, B=ScalarParameter() {defaultValue=-0.3, group=A_L3 Custom Mask, parameterName="• L3_WS Noise_Bias", sortPriority=106})
  n398 = ScalarParameter() {defaultValue=5, group=A_L3 Custom Mask, parameterName="• L3_WS Noise_Scale", sortPriority=106}
  NamedRerouteDeclaration(None=Saturate(None=Multiply(A=@n397, B=@n398))) {name=L3_M_WS noise, nodeColor=<rgba 0 0.5898 1 1>}
  n399 = Add(A=NamedRerouteUsage(), B=ScalarParameter() {group=A_L4 Landscape Mask, parameterName="• L4_L1 Height_Bias", sortPriority=116})
  n400 = ScalarParameter() {defaultValue=1, group=A_L4 Landscape Mask, parameterName="• L4_L1 Height_Contrast", sortPriority=116}
  NamedRerouteDeclaration(None=Saturate(None=Multiply(A=@n399, B=@n400))) {name=L4_M_L1H, nodeColor=<rgba 0.5 0 1 1>}
  NamedRerouteDeclaration(None=@n386.Normal) {name=L2_N, nodeColor=<rgba 0.3073 1 0.5732 1>}
  n401 = ScalarParameter() {defaultValue=3.2, group=A_L3 Custom Mask, parameterName="• L3 Height Mask_Contrast", sortPriority=595}
  n402 = Subtract(B=@n401) {constA=0, desc=Low negative value}
  n403 = Add(A=@n401) {desc=High Positive}
  n404 = Add(
   A=NamedRerouteUsage()
   B=ScalarParameter() {defaultValue=-0.5, group=A_L3 Custom Mask, parameterName="• L3 Height Mask_Bias", sortPriority=595}
  )
  NamedRerouteDeclaration(None=Saturate(None=LinearInterpolate(A=@n402, B=@n403, Alpha=@n404) {desc=Lerps between greater range of value})) {name=L3_Masks_ Height, nodeColor=<rgba 0 0.5898 1 1>}
  n405 = ChannelMaskParameter(None=@n284) {group="A_L2 (TD slot 2)", maskChannel=Green, parameterName=Curvature_Channel, sortPriority=550, defaultValue=<rgba 0 1 0 0>}
  NamedRerouteDeclaration(None=@n405.R) {name=L2 Curvature, nodeColor=<rgba 0.2187 1 0 1>}
  n406 = Add(A=NamedRerouteUsage(), B=ScalarParameter() {group=A_L3 Custom Mask, parameterName="• L3_Painted_Bias", sortPriority=56})
  n407 = ScalarParameter() {defaultValue=1, group=A_L3 Custom Mask, parameterName="• L3_Painted_Scale", sortPriority=56}
  NamedRerouteDeclaration(None=Saturate(None=Multiply(A=@n406, B=@n407))) {name=L3_Painted, nodeColor=<rgba 0 0.5898 1 1>}
  n408 = ChannelMaskParameter(None=@n284) {group="A_L2 (TD slot 2)", maskChannel=Blue, parameterName=Painted Mask_Channel, sortPriority=550, defaultValue=<rgba 0 0 1 0>}
  NamedRerouteDeclaration(None=@n408.R) {name=L2_Painted Mask, nodeColor=<rgba 0 0 1 1>}
  n409 = NamedRerouteUsage()
  n410 = NamedRerouteUsage()
  n411 = NamedRerouteUsage()
  n412 = Add(
   A=ChannelMaskParameter(None=MaterialFunctionCall(R=@n409, G=@n410, B=@n411) {materialFunction=&AppendMany}.RGB) {group=A_L3 Custom Mask, parameterName=L3_L1 SMR_Channel, sortPriority=125}.R
   B=ScalarParameter() {group=A_L3 Custom Mask, parameterName="• L3_L1 SMR_Bias", sortPriority=126}
  )
  n413 = ScalarParameter() {defaultValue=1, group=A_L3 Custom Mask, parameterName="• L3_L1 SMR_Contrast", sortPriority=126}
  NamedRerouteDeclaration(None=Saturate(None=Multiply(A=@n412, B=@n413))) {name=L3_M_L1 SMR, nodeColor=<rgba 0 0.5898 1 1>}
  NamedRerouteDeclaration(None=SetMaterialAttributes!(MaterialAttributes, Roughness=Constant() {r=0.9}) {attributeSetTypes [#7]}) {name=Debug_MA, nodeColor=<rgba 0.2422 0 1 1>}
  n414 = TextureSampleParameter2D!(UVs=@n117, Apply View MipBias) {
   group=CustomCrack
   parameterName=CustomCrack_Mask
   samplerSource=SSM_Wrap_WorldGroupSettings
   samplerType=SAMPLERTYPE_LinearGrayscale
   sortPriority=-100
   texture=/Game/Global/Textures/Default/Blanks/T_BlackLinearGrayscale
  }
  n415 = TextureSampleParameter2D(UVs=@n123) {
   group=CustomCrack
   parameterName=CustomCrack_Mask
   samplerSource=SSM_Wrap_WorldGroupSettings
   samplerType=SAMPLERTYPE_LinearGrayscale
   sortPriority=-100
   texture=&T_BlackLinearGrayscale
  }
  n416 = TextureSampleParameter2D(UVs=@n126) {
   group=CustomCrack
   parameterName=CustomCrack_Mask
   samplerSource=SSM_Wrap_WorldGroupSettings
   samplerType=SAMPLERTYPE_LinearGrayscale
   sortPriority=-100
   texture=&T_BlackLinearGrayscale
  }
  n417 = Max(A=@n415.R, B=@n416.R)
  n418 = DDX(Value=@n117)
  n419 = DDY(Value=@n117)
  n420 = TextureSampleParameter2D!(UVs=@n123, "DDX(UVs)"=@n418, "DDY(UVs)"=@n419, Apply View MipBias) {
   group=CustomCrack
   mipValueMode=TMVM_Derivative
   parameterName=CustomCrack_Mask
   samplerSource=SSM_Wrap_WorldGroupSettings
   samplerType=SAMPLERTYPE_LinearGrayscale
   sortPriority=-100
   texture=&T_BlackLinearGrayscale
  }
  n421 = TextureSampleParameter2D(UVs=@n126, "DDX(UVs)"=@n418, "DDY(UVs)"=@n419) {
   group=CustomCrack
   mipValueMode=TMVM_Derivative
   parameterName=CustomCrack_Mask
   samplerSource=SSM_Wrap_WorldGroupSettings
   samplerType=SAMPLERTYPE_LinearGrayscale
   sortPriority=-100
   texture=&T_BlackLinearGrayscale
  }
  n422 = Max(A=@n420.R, B=@n421.R)
  n423 = MaterialFunctionCall(
   "Xb1,PS4"=@n414.R
   "PC-High&&Epic,XSX,PS5"=Multiply(A=NaniteReplace(Default=@n417, Nanite=@n422), B=NamedRerouteUsage())
   "PC-Perf,Android"=@n414.R
   Switch=@n414.R
   "PC-Low(NotPerf)&&Medium"=@n414.R
  ) {materialFunction=&PlatformSwitch_1D}
  n424 = Power(
   Base=MaterialFunctionCall(In=@n423.Result, Contrast=ScalarParameter() {defaultValue=1, group=CustomCrack, parameterName=CustomCrack_MaskContrast}) {materialFunction=&CheapContrast}.Result
   Exp=ScalarParameter() {defaultValue=32, group=CustomCrack, parameterName=CustomCrack_MaskPower}
  )
  n425 = Power(Base=@n423.Result, Exp=ScalarParameter() {defaultValue=3, group=CustomCrack, parameterName=CrackBreakup_PowerBase})
  n426 = TextureSample!(
   UVs=Multiply(A=TextureCoordinate(), B=ScalarParameter() {defaultValue=96, group=CustomCrack, parameterName=CrackBreakup_Tiling})
   Tex
   Apply View MipBias
  ) {samplerSource=SSM_Wrap_WorldGroupSettings, samplerType=SAMPLERTYPE_LinearColor, texture=/Game/Global/Textures/Default/Noise/T_CloudNoise_01}
  n427 = MaterialFunctionCall(
   In=Power(Base=@n426.R, Exp=ScalarParameter() {defaultValue=1, group=CustomCrack, parameterName=CrackBreakup_Power}) {constExponent=1}
   Contrast=ScalarParameter() {defaultValue=-0.6, group=CustomCrack, parameterName=CrackBreakup_Contrast}
  ) {materialFunction=&CheapContrast}
  n428 = Multiply(
   A=Subtract(
    A=MaterialFunctionCall(Base=@n425, Blend=@n427.Result) {materialFunction=&Blend_Overlay}.Result
    B=@n425
   )
   B=ScalarParameter() {defaultValue=16, group=CustomCrack, parameterName=CrackBreakup_Intensity}
  )
  n429 = Max(
   A=@n424
   B=MaterialFunctionCall(In=Saturate(None=@n428), Contrast=Constant() {r=512}) {materialFunction=&CheapContrast}.Result
  )
  NamedRerouteDeclaration(None=MaterialFunctionCall("Xb1,PS4"=@n429, "PC-High&&Epic,XSX,PS5"=@n429, "PC-Perf,Android"=@n424, Switch=@n424, "PC-Low(NotPerf)&&Medium"=@n424) {materialFunction=&PlatformSwitch_1D}.Result) {name=Crack_GoldMask, nodeColor=<rgba 1 0.5156 0 1>}
  n430 = NamedRerouteUsage()
  n431 = NamedRerouteUsage()
  n432 = NamedRerouteUsage()
  n433 = Add(
   A=ChannelMaskParameter(None=MaterialFunctionCall(R=@n430, G=@n431, B=@n432) {materialFunction=&AppendMany}.RGB) {group=A_L4 Landscape Mask, parameterName=L4_L1 SMR_Channel, sortPriority=125}.R
   B=ScalarParameter() {group=A_L4 Landscape Mask, parameterName="• L4_L1 SMR_Bias", sortPriority=126}
  )
  n434 = ScalarParameter() {defaultValue=1, group=A_L4 Landscape Mask, parameterName="• L4_L1 SMR_Contrast", sortPriority=126}
  NamedRerouteDeclaration(None=Saturate(None=Multiply(A=@n433, B=@n434))) {name=L4_M_L1 SMR, nodeColor=<rgba 0.5 0 1 1>}
  n435 = NamedRerouteUsage()
  n436 = NamedRerouteUsage()
  NamedRerouteDeclaration(None=MaterialFunctionCall!(Linear=Multiply(A=@n435, B=@n436)) {materialFunction=/Engine/Functions/Engine_MaterialFunctions01/ImageAdjustment/LinearTosRGB}.sRGB) {name=L1 H * L2 Curvature, nodeColor=<rgba 0.4766 1 0 1>}
  NamedRerouteDeclaration(None=@n116.GrassLayer) {name=Landscape_Grass, nodeColor=<rgba 0.0312 0 1 1>}
  NamedRerouteDeclaration(None=@n161.Roughness) {name=L1_Roughness, nodeColor=<rgba 1 0.5625 0 1>}
  NamedRerouteDeclaration(None=@n386.OpacityMask) {name=L2_H, nodeColor=<rgba 0 1 0.3839 1>}
  NamedRerouteDeclaration(None=@n263.OpacityMask) {name=Landscape_H, nodeColor=<rgba 0 0.7891 1 1>}
  NamedRerouteDeclaration(None=ComponentMask(None=@n277) {a=true}) {name=L3_H, nodeColor=<rgba 1 0.8906 0 1>}
  NamedRerouteDeclaration(None=@n116.DesertGrassLayer) {name=Landscape_DesertGrass, nodeColor=<rgba 0.0312 0 1 1>}
  NamedRerouteDeclaration(None=@n116.GravelLayer) {name=Landscape_Gravel, nodeColor=<rgba 1 0.8662 0.519 1>}
  NamedRerouteDeclaration(None=@n116.SandLayer) {name=Landscape_Sand, nodeColor=<rgba 0 0.9297 1 1>}
  NamedRerouteDeclaration(None=@n116.AsphaltLayer) {name=Landscape_Asphalt, nodeColor=<rgba 1 0.75 0 1>}
  n437 = NamedRerouteUsage()
  n438 = ComponentMask(None=VertexNormalWS()) {b=true}
  NamedRerouteDeclaration(None=Saturate(None=Max(A=@n437, B=@n438))) {name=Fade From Center To Top, nodeColor=<rgba 0 0.1562 1 1>}
  NamedRerouteDeclaration(None=@n116.ScreeLayer) {name=Landscape_Scree, nodeColor=<rgba 0.4297 0 1 1>}
  NamedRerouteDeclaration(None=@n116.SnowLayer) {name=Landscape_Snow, nodeColor=<rgba 0 1 0.1094 1>}
  n439 = ScalarParameter() {defaultValue=16, group=CustomCrack, parameterName=CustomCrack_Crack_MaskMult, sortPriority=-90}
  NamedRerouteDeclaration(None=Saturate(None=Multiply(A=@n423.Result, B=@n439))) {name=CrackMask, nodeColor=<rgba 1 0.5156 0 1>}
  n440 = Add(
   A=StaticSwitchParameter(True=@n304, False=@n306) {desc="T= VTX only\r\nF= L2 Normals", group=A_L4 Landscape Mask, parameterName=L4_World Z Normal_Use VTX, sortPriority=15}
   B=ScalarParameter() {defaultValue=-0.2, group=A_L4 Landscape Mask, parameterName="• L4_World Z Normal_Bias", sortPriority=15}
  )
  n441 = ScalarParameter() {defaultValue=1.5, group=A_L4 Landscape Mask, parameterName="• L4_World Z Normal_Scale", sortPriority=15}
  NamedRerouteDeclaration(None=Saturate(None=Multiply(A=@n440, B=@n441))) {name=L4_Normal Z, nodeColor=<rgba 0.5 0 1 1>}
  n442 = VertexColor()
  n443 = Constant() {r=2}
  n444 = ScalarParameter() {
   defaultValue=0.5
   desc=Modulates VTX color transition intensity
   group=A_L4 Landscape Mask
   parameterName="• L4_VTX Color_Contrast"
   sortPriority=26
  }
  n445 = DotProduct(
   A=Multiply(A=ChannelMaskParameter(None=@n442) {group=A_L4 Landscape Mask, parameterName="• L4_VTX Color_Channel", sortPriority=25}.R) {constB=3}
   B=ShadingPathSwitch(Default=@n444, Mobile=Multiply(A=@n444, B=@n443) {desc=compensates for the less dense meshes on mobile creating bigger vertex gradients})
  )
  NamedRerouteDeclaration(None=VertexInterpolator(VS=Saturate(None=@n445)).PS) {name=L4_Vertex Color, nodeColor=<rgba 0.5 0 1 1>}
  n446 = StaticSwitchParameter(True=NamedRerouteUsage(), False=NamedRerouteUsage()) {
   desc="** DF AO is for Epic Settings Only!True= AO Baked from VTX or Texture\r\nFalse= AO Distance Field"
   group=A_L4 Landscape Mask
   parameterName=L4_AO_Force Baked on Nanite
   sortPriority=35
  }
  n447 = Add(A=@n446, B=ScalarParameter() {group=A_L4 Landscape Mask, parameterName="• L4_AO_Bias", sortPriority=36})
  n448 = ScalarParameter() {defaultValue=1, group=A_L4 Landscape Mask, parameterName="• L4_AO_Contrast", sortPriority=36}
  NamedRerouteDeclaration(None=Saturate(None=Multiply(A=@n447, B=@n448))) {name=L4_AO Mask, nodeColor=<rgba 0.5 0 1 1>}
  n449 = Add(A=NamedRerouteUsage(), B=ScalarParameter() {group=A_L4 Landscape Mask, parameterName="• L4_Curvature_Bias", sortPriority=46})
  n450 = ScalarParameter() {defaultValue=1, group=A_L4 Landscape Mask, parameterName="• L4_Curvature_Scale", sortPriority=46}
  NamedRerouteDeclaration(None=Saturate(None=Multiply(A=@n449, B=@n450))) {name=L4_Curvature/Thickness, nodeColor=<rgba 0.5 0 1 1>}
  n451 = Add(A=NamedRerouteUsage(), B=ScalarParameter() {group=A_L4 Landscape Mask, parameterName="• L4_Painted_Bias", sortPriority=56})
  n452 = ScalarParameter() {defaultValue=1, group=A_L4 Landscape Mask, parameterName="• L4_Painted_Scale", sortPriority=56}
  NamedRerouteDeclaration(None=Saturate(None=Multiply(A=@n451, B=@n452))) {name=L4_Painted, nodeColor=<rgba 0.5 0 1 1>}
  n453 = ScalarParameter() {desc="1= object top bound\r\n0= object bottom bound", group=A_L4 Landscape Mask, parameterName="• L4_Z Gradient_HBottom", sortPriority=63}
  n454 = ScalarParameter() {
   defaultValue=1
   desc="1= object top bound\r\n0= object bottom bound"
   group=A_L4 Landscape Mask
   parameterName="• L4_Z Gradient_HTop"
   sortPriority=62
  }
  NamedRerouteDeclaration(None=SmoothStep(Min=@n453, Max=@n454, Value=NamedRerouteUsage())) {name=L4_Z Gradient, nodeColor=<rgba 0.3828 0 1 1>}
  NamedRerouteDeclaration(None=SmoothStep(
    Min=ScalarParameter() {group=A_L4 Landscape Mask, parameterName="• L4_Waterline_Min", sortPriority=75}
    Max=ScalarParameter() {defaultValue=0.5, group=A_L4 Landscape Mask, parameterName="• L4_Waterline_Max", sortPriority=75}
    Value=NamedRerouteUsage()
   )) {name=L4_Waterline, nodeColor=<rgba 0.5 0 1 1>}
  n455 = Multiply(
   A=Add(A=NamedRerouteUsage(), B=ScalarParameter() {defaultValue=9.8, group=CustomCrack, parameterName=CrackBreakup_RVTBlend_Offset})
   B=ScalarParameter() {defaultValue=0.05, group=CustomCrack, parameterName=CrackBreakup_RVTBlend_Size}
  ) {constB=0.009}
  n456 = ScalarParameter() {defaultValue=1.5, group=CustomCrack, parameterName=CrackBreakup_RVTBlend_Power}
  NamedRerouteDeclaration(None=Saturate(None=Power(Base=@n455, Exp=@n456))) {name=Crack_RVTBlend, nodeColor=<rgba 1 0.8672 0 1>}
  n457 = Constant() {r=75}
  n458 = Constant() {r=30}
  n459 = MaterialFunctionCall(
   Texture Scale=ScalarParameter() {defaultValue=2000, group=A_L4 Landscape Mask, parameterName="• L4_WS Noise Tile", sortPriority=106}
   Triplanar Fade=MaterialFunctionCall(Low=@n457, High=@n458, Epic=@n458, --------=@n458, Mobile=@n457, Medium=@n458) {materialFunction=&MF_QualitySwitch_Scalar}.Result
  ) {materialFunction=&MF_TriplanarDithered}
  n460 = TextureSampleParameter2D(UVs=@n459.UVs, "DDX(UVs)"=@n459.UVs, "DDY(UVs)"=@n459.UVs) {
   group=A_L4 Landscape Mask
   mipValueMode=TMVM_Derivative
   parameterName="• L4_WS Noise_WorldSpaceNoise"
   sortPriority=105
   texture=&TilingNoise
  }
  n461 = Add(A=@n460.RGB, B=ScalarParameter() {defaultValue=-0.5, group=A_L4 Landscape Mask, parameterName="• L4_WS Noise_Bias", sortPriority=106})
  n462 = ScalarParameter() {defaultValue=1, group=A_L4 Landscape Mask, parameterName="• L4_WS Noise_Scale", sortPriority=106}
  NamedRerouteDeclaration(None=Saturate(None=Multiply(A=@n461, B=@n462))) {name=L4_M_WS noise, nodeColor=<rgba 1 0 0.8906 1>}
  NamedRerouteDeclaration() {name=Landscape_MossyForest, nodeColor=<rgba 0 0.5312 1 1>}
  NamedRerouteDeclaration(None=@n321.Normal) {name=L1_Normal_1, nodeColor=<rgba 0.9922 0 1 1>}
  NamedRerouteDeclaration(None=@n321.Roughness) {name=L1_Rough_1, nodeColor=<rgba 0.9922 0 1 1>}
  NamedRerouteDeclaration(None=@n321.Specular) {name=L1_Spec_1, nodeColor=<rgba 0.9922 0 1 1>}
  NamedRerouteDeclaration(None=@n321.BaseColor) {name=L1_Basecolor_1, nodeColor=<rgba 0.9922 0 1 1>}
  NamedRerouteDeclaration(None=@n321.MaterialAttributes) {name=L1_Material_1, nodeColor=<rgba 0.9922 0 1 1>}
  NamedRerouteDeclaration(None=@n65) {name=L3CustomTop, nodeColor=<rgba 0 0.2969 1 1>}
  n463 = StaticBoolParameter() {desc="If set to True, set TopLayer to False", group=0_Rock/Cliff Main Settings, parameterName=L4_Use Landscape Material, sortPriority=9}
  n464 = NamedRerouteDeclaration(None=@n463) {name=Use Landscape Material, nodeColor=<rgba 1 0.4219 0 1>}
  n465 = NamedRerouteUsage()
  n466 = BlendMaterialAttributes(A=@n465, B=NamedRerouteUsage(), Alpha=NamedRerouteUsage()) {vertexAttributeBlendType=UseB}
  NamedRerouteDeclaration(None=StaticSwitchParameter(True=NamedRerouteUsage(), False=StaticSwitch(True=@n466, False=@n465, Value=@n464)) {group=0_Rock/Cliff Main Settings, parameterName=L4_Use Landscape Material Only, sortPriority=10}) {name=CombinedLayers, nodeColor=<rgba 0.5 0 1 1>}
  n467 = NamedRerouteUsage()
  n468 = TextureObject() {
   desc="T_Hermes_MaskMaker_Rock\r\n"
   samplerType=SAMPLERTYPE_LinearGrayscale
   texture=&T_BlackLinearGrayscale
  }
  n469 = ScalarParameter() {defaultValue=4, group=CustomCrack, parameterName=CustomCrack_NMI, sortPriority=700}
  n470 = ScalarParameter() {defaultValue=0.001, group=CustomCrack, parameterName=CustomCrack_HMO, sortPriority=800}
  n471 = TextureCoordinate()
  n472 = LinearInterpolate(
   A=@n467
   B=Multiply(
    A=MaterialFunctionCall!(Height Map=@n468, Normal Map Intensity=@n469, Height Map UV Offset=@n470, Coordinates=@n471, Height Map Channel Selector) {materialFunction=/Engine/Functions/Engine_MaterialFunctions03/Procedurals/NormalFromHeightmap}.Result
    B=Constant3Vector() {constant=<rgba -1 -1 1 0>}
   )
   Alpha=ScalarParameter() {defaultValue=1, group=CustomCrack, parameterName=CustomCrack_TestNormalMethod, sortPriority=900}
  )
  n473 = StaticBoolParameter() {desc=Only select one option per layer as true, group=CustomCrack, parameterName=Debug_NormalFromHeightmap, sortPriority=600}
  NamedRerouteDeclaration(None=Transform(None=StaticSwitch(True=@n472, False=@n467, Value=@n473))) {name=CrackNormals, nodeColor=<rgba 1 0.5156 0 1>}
  NamedRerouteUsage()
  NamedRerouteUsage()
  NamedRerouteUsage()
  NamedRerouteUsage()
  NamedRerouteUsage()
  NamedRerouteUsage()
  NamedRerouteUsage()
  NamedRerouteUsage()
  NamedRerouteUsage()
  NamedRerouteUsage()
  NamedRerouteUsage()
  NamedRerouteUsage()
  PinBase!(Output=@n50) {reroutePins [Output {expression=&self:MaterialExpressionReroute_98}]}
  PinBase!() {reroutePins [MaterialAttributes {expression=&self:MaterialExpressionReroute_70}]}
  PinBase() {reroutePins [MaterialAttributes {expression=&self:MaterialExpressionReroute_99}]}
  PinBase!(Result=@n54.Result) {reroutePins [Result {expression=&self:MaterialExpressionReroute_100}]}
  n474 = Transform(None=GetMaterialAttributes(None=NamedRerouteUsage()) {attributeGetTypes [#6]}.Normal)
  n475 = VertexNormalWS()
  Reroute(None=LinearInterpolate(A=@n474, B=@n475, Alpha=ComponentMask(None=@n475) {b=true}))
  Reroute(None=Multiply(A=NamedRerouteUsage(), B=NamedRerouteUsage()))
  Reroute(None=@n474)
  n476 = TextureObjectParameter() {
   group=Goldlands
   parameterName=Goldlands_GoldMask
   samplerType=SAMPLERTYPE_Masks
   sortPriority=-500
   texture=&T_BlackMask
  }
  n477 = GetMaterialAttributes(None=@n39.Result) {attributeGetTypes [#1, #9, #6]}
  n478 = Multiply(A=@n477.BaseColor, B=VectorParameter() {group=Metal_Cubemap, parameterName=Metal_Cubemap_Mult, defaultValue=<rgba 10 7.4272 2.5 1>}.RGB)
  n479 = TextureSampleParameter2D!(
   UVs=MaterialFunctionCall!(Vector=ReflectionVectorWS()) {materialFunction=/Engine/Functions/Engine_MaterialFunctions01/Coordinates/LongLatToUV}.Result
   Apply View MipBias
  ) {
   group=Metal_Cubemap
   parameterName=Metal_LongLat_Texture
   samplerSource=SSM_Wrap_WorldGroupSettings
   texture=/Game/Packages/Fortress_SharedMaterials/Textures/T_Ravens_SpecularReflection
  }
  n480 = Add(A=@n477.BaseColor, B=Multiply(A=Multiply(A=@n477.Metallic, B=@n478), B=@n479.RGB))
  n481 = StaticSwitch(
   True=SetMaterialAttributes!(MaterialAttributes=@n477.MaterialAttributes, BaseColor=Saturate(None=@n480)) {attributeSetTypes [#1]}
   False=@n39.Result
   Value=StaticBoolParameter() {desc=Only select one option per layer as true, group=Metal_Cubemap, parameterName=Use_Metal_Cubemap, sortPriority=-1000}
  )
  n482 = StaticSwitch(
   True=MaterialFunctionCall!(Material=@n481, MobileMaterial=@n481, Texture Mask=@n476, Texture UVs=TextureCoordinate(), IsStaticMesh=StaticBool() {value=true}) {materialFunction=/BRCoreMaterials/Hermes/Terrain/Landscape/Materials/Layers/MF_Hermes_Goldlands}.BlendedMaterial
   False=@n481
   Value=StaticBoolParameter() {desc=Only select one option per layer as true, group=Goldlands, parameterName=Use_Goldlands, sortPriority=-1000}
  )
  n483 = GetMaterialAttributes(None=@n482) {attributeGetTypes [
    #1
    #8
    #7
    #9
   ]}
  n484 = VectorParameter() {group=CustomCrack, parameterName=CustomCrack_GoldColor, sortPriority=-50, defaultValue=<rgba 2 1.2116 0.5217 0.4>}
  n485 = VectorParameter() {
   desc=Dot product of the basecolor before the gold color multiply
   group=CustomCrack
   parameterName=CustomCrack_BasecolorDotProduct
   sortPriority=-60
   defaultValue=<rgba 0.1 0.1 3 1>
  }
  n486 = Multiply(A=@n484.RGB, B=DotProduct(A=@n485.RGB, B=@n483.BaseColor))
  n487 = NamedRerouteUsage()
  n488 = VectorParameter() {
   group=CustomCrack
   parameterName=CustomCrack_RoughnessSettings
   channelNames {a="", b=CustomCrack_GoldRoughness_Max, g=CustomCrack_GoldRoughness_Min, r=CustomCrack_GoldRoughness_Power}
   defaultValue=<rgba 3 0.1 0.4 1>
  }
  n489 = NamedRerouteUsage()
  n490 = LinearInterpolate(
   A=@n488.CustomCrack_GoldRoughness_Min
   B=@n488.CustomCrack_GoldRoughness_Min
   Alpha=Power(
    Base=GetMaterialAttributes(None=@n489) {attributeGetTypes [#7]}.Roughness
    Exp=@n488.CustomCrack_GoldRoughness_Power
   )
  )
  n491 = SetMaterialAttributes!(
   MaterialAttributes=@n483.MaterialAttributes
   BaseColor=LinearInterpolate(A=@n483.BaseColor, B=Saturate(None=@n486), Alpha=@n487)
   Metallic=Max(A=@n487, B=@n483.Metallic)
   Normal=NamedRerouteUsage()
   Specular=@n483.Specular
   Roughness=LinearInterpolate(A=@n483.Roughness, B=Saturate(None=@n490), Alpha=NamedRerouteUsage())
  ) {attributeSetTypes [
    #1
    #9
    #6
    #8
    #7
   ]}
  n492 = NamedRerouteUsage()
  Reroute(None=StaticSwitch(
    True=BlendMaterialAttributes(A=@n482, B=@n491, Alpha=@n492) {vertexAttributeBlendType=UseB}
    False=@n482
    Value=StaticBoolParameter() {desc=Only select one option per layer as true, group=CustomCrack, parameterName=Use_CustomCrack, sortPriority=-1000}
   ))
  Reroute(None=@n181)
  Reroute(None=@n195)
  n493 = ScalarParameter() {group="2_Dirt/Curvature (L2_MSK as input)", parameterName="Use Non-Epic Colors? (Curvature/Dirt)", sliderMax=1, sortPriority=1}
  Saturate(None=@n493)
  n494 = VectorParameter() {
   parameterName="Dirt_AO Color ( Expand for blending option)"
   sortPriority=100031
   channelNames {a=Linear vs Mult, b="", g="", r=""}
   defaultValue=<rgba 0.2691 0.1783 0.1117 0.5>
  }
  n495 = Constant() {r=1}
  n496 = ScalarParameter() {
   defaultValue=4
   desc="Multiplies SpecularMasks_2 G Channel (L2 thickness mask)"
   parameterName=Dirt_Edge Mask Multiplier
   sortPriority=100030
  }
  n497 = Multiply(A=@n496, B=@n495) {constA=4}
  n498 = NamedRerouteUsage()
  n499 = GetMaterialAttributes() {attributeGetTypes [#1, #8]}
  n500 = Multiply(A=Multiply(A=LinearInterpolate(A=@n495, B=@n497, Alpha=@n498), B=@n499.BaseColor), B=NamedRerouteUsage())
  n501 = Multiply(A=@n494, B=@n500)
  n502 = Multiply(A=LinearInterpolate(A=@n499.Specular, Alpha=NamedRerouteUsage()), B=NamedRerouteUsage())
  SetMaterialAttributes!(
   MaterialAttributes=@n499.MaterialAttributes
   BaseColor=LinearInterpolate(A=LinearInterpolate(A=@n501, B=@n494, Alpha=@n494), B=@n500, Alpha=Saturate(None=@n502)) {constA=1}
   Specular
   Roughness
   Normal
  ) {attributeSetTypes [
    #1
    #8
    #7
    #6
   ]}
  n503 = GetMaterialAttributes(None=@n55) {attributeGetTypes [#1]}
  SetMaterialAttributes!(
   MaterialAttributes=@n503.MaterialAttributes
   BaseColor=LinearInterpolate(A=@n503.BaseColor, Alpha=ScalarParameter() {parameterName=DEBUG_TestBasecolor})
  ) {attributeSetTypes [#1]}
  StaticSwitch(True=NamedRerouteUsage(), Value=StaticBoolParameter() {group=A_L4_Landscape, parameterName=Top Mossy Forest, sortPriority=-4})
  StaticSwitch(True=@n277, False=@n276, Value=NamedRerouteUsage())
  StaticSwitch(True=NamedRerouteUsage(), False=NamedRerouteUsage(), Value=NamedRerouteUsage())
  n504 = StaticBoolParameter() {
   defaultValue=true
   desc="On: L2 Use UV0\r\nOff: L2 Use UV1"
   group=0_Rock/Cliff Main Settings
   parameterName=L2 Use Displacement
   sortPriority=5
  }
  StaticSwitch(True=TextureCoordinate(), False=TextureCoordinate() {coordinateIndex=1}, Value=@n504)
  StaticSwitch(True=NaniteReplace(), Value=NamedRerouteUsage())
  VertexInterpolator(VS=@n290) {desc="Would be move 72 instructions to vertex shader  but currently causes artifacs in old meshes.\r\nTo add this it would be neessary:\r\n• have a mesh revision to better handle decimation on nanite meshes.\r\n• adding it behind a static switch in case we need to use it in pixel shader on specific assets"}
 ]
}
