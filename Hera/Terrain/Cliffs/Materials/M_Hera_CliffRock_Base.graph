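# Base material for Hera terrain cliff rock, built for Nanite and instanced static meshes as Default Lit.
# All surface logic lives in MF_Hera_CliffRock, whose material attributes pass through CINE_MF_EnvPostFX and then MF_CreativeMoveTool to the output.
# Normals are not in tangent space, and displacement is enabled with a screen-size-based fade.
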Material /Hera/Terrain/Cliffs/Materials/M_Hera_CliffRock_Base
graph {
 properties {
  bEnableDisplacementFade=true
  bTangentSpaceNormal=false
  bUseMaterialAttributes=true
  bUsedWithInstancedStaticMeshes=true
  bUsedWithNanite=true
  pixelDepthOffsetMode=PDOM_Legacy
  usedShadingModels=Default Lit
  displacementFadeRange {endSizePixels=6, startSizePixels=8}
  displacementScaling {center=0.5, magnitude=50}
 }
 nodes [MaterialFunctionCall(MA In=MaterialFunctionCall!(In=MaterialFunctionCall() {materialFunction=&/MF_Hera_CliffRock}.Result) {materialFunction=/Game/Packages/Fortress_SharedMaterials/Cinematics/CINE_MF_EnvPostFX}.Result) {materialFunction=&MF_CreativeMoveTool}]
}
